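# Physics asset for the F_MED_MotorMonth female medium player body, covering a three-strand cape and the upper body around it.
# Left, middle and right dyn_cape capsule chains hang from small MSTR root bodies constrained to spine_05, with the middle chain one segment longer.
# Capsules on the upper spine and upper arms provide body volumes.
# Cape segment joints have angular limits with tight twist, while spine, arm and cape-root joints swing and twist freely.
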PhysicsAsset /BRCosmetics/Characters/Player/Female/Medium/Bodies/F_MED_MotorMonth/Meshes/F_MED_MotorMonth_Physics
bodies {
 dyn_cape_l_01 [{shapeType=Capsule, length=4, radius=5, shapeName=dyn_cape_l_01_capsule, center=<1 -3.5 0>, rotation=<rot 90 0 0>}]
 dyn_cape_l_02 [{shapeType=Capsule, length=4, radius=6, shapeName=dyn_cape_l_02_capsule, center=<1 -3.5 0>, rotation=<rot 90 0 0>}]
 dyn_cape_l_MSTR [{shapeType=Capsule, length=0.505, radius=0.505, shapeName=dyn_cape_l_MSTR_capsule, rotation=<rot 0 -90 0>}]
 dyn_cape_mid_01 [{shapeType=Capsule, length=2.5, radius=6, shapeName=dyn_cape_mid_01_capsule, center=<1 -3.5 0>, rotation=<rot 90 0 0>}]
 dyn_cape_mid_02 [{shapeType=Capsule, length=9.1113, radius=5.5584, shapeName=dyn_cape_mid_02_capsule, center=<4.0031 -4 0>, rotation=<rot 90 0 0>}]
 dyn_cape_mid_03 [{shapeType=Capsule, length=3.4377, radius=1.8265, shapeName=dyn_cape_mid_03_capsule, center=<3 -0.5 0>}]
 dyn_cape_mid_MSTR [{shapeType=Capsule, length=0.505, radius=0.505, shapeName=dyn_cape_mid_MSTR_capsule, rotation=<rot 0 -90 0>}]
 dyn_cape_r_01 [{shapeType=Capsule, length=4, radius=5, shapeName=dyn_cape_r_01_capsule, center=<1 -3.5 0>, rotation=<rot 90 0 0>}]
 dyn_cape_r_02 [{shapeType=Capsule, length=4, radius=6, shapeName=dyn_cape_r_02_capsule, center=<1 -3.5 0>, rotation=<rot 90 0 0>}]
 dyn_cape_r_MSTR [{shapeType=Capsule, length=0.505, radius=0.505, shapeName=dyn_cape_r_MSTR_capsule, rotation=<rot 0 -90 0>}]
 spine_03 [{shapeType=Capsule, length=10.1899, radius=7.3917, shapeName=spine_03_capsule, rotation=<rot 0 -90 0>}]
 spine_04 [{
   shapeType=Capsule
   length=14.0174
   radius=9.4781
   shapeName=spine_04_capsule
   center=<-2.4597 3.4452 -0.6492>
   rotation=<rot 4.7349 -90 -2.1845>
  }]
 spine_05 [{shapeType=Capsule, length=25.073, radius=10.1147, center=<-2.4144 2.2744 0>, rotation=<rot 0 179.9999 180>}]
 upperarm_l [{shapeType=Capsule, length=14.081, radius=5.3399, center=<7.0817 0.4636 0>, rotation=<rot 90 -172.875 -162.875>}]
 upperarm_r [{shapeType=Capsule, length=14.0559, radius=5.4003, center=<-7.0802 -0.4551 0>, rotation=<rot 90 179.2838 -170.7165>}]
}
constraints [
 {
  bone1Name=upperarm_l
  bone2Name=spine_05
  swing1LimitDegrees=45
  swing1Motion=Free
  swing2LimitDegrees=45
  swing2Motion=Free
  twistLimitDegrees=45
  twistMotion=Free
 }
 {
  bone1Name=upperarm_r
  bone2Name=spine_05
  swing1LimitDegrees=45
  swing1Motion=Free
  swing2LimitDegrees=45
  swing2Motion=Free
  twistLimitDegrees=45
  twistMotion=Free
 }
 {bone1Name=dyn_cape_mid_02, bone2Name=dyn_cape_mid_01, swing1LimitDegrees=45, swing2LimitDegrees=30, twistLimitDegrees=2}
 {
  bone1Name=spine_05
  bone2Name=spine_04
  swing1LimitDegrees=45
  swing1Motion=Free
  swing2LimitDegrees=45
  swing2Motion=Free
  twistLimitDegrees=45
  twistMotion=Free
 }
 {bone1Name=dyn_cape_r_02, bone2Name=dyn_cape_r_01, swing1LimitDegrees=45, swing2LimitDegrees=30, twistLimitDegrees=2}
 {bone1Name=dyn_cape_mid_01, bone2Name=dyn_cape_mid_MSTR, swing1LimitDegrees=45, swing2LimitDegrees=30, twistLimitDegrees=2}
 {bone1Name=dyn_cape_r_01, bone2Name=dyn_cape_r_MSTR, swing1LimitDegrees=45, swing2LimitDegrees=30, twistLimitDegrees=2}
 {bone1Name=dyn_cape_l_02, bone2Name=dyn_cape_l_01, swing1LimitDegrees=45, swing2LimitDegrees=30, twistLimitDegrees=2}
 {bone1Name=dyn_cape_l_01, bone2Name=dyn_cape_l_MSTR, swing1LimitDegrees=45, swing2LimitDegrees=30, twistLimitDegrees=2}
 {
  bone1Name=dyn_cape_r_MSTR
  bone2Name=spine_05
  swing1LimitDegrees=45
  swing1Motion=Free
  swing2LimitDegrees=45
  swing2Motion=Free
  twistLimitDegrees=45
  twistMotion=Free
 }
 {
  bone1Name=dyn_cape_mid_MSTR
  bone2Name=spine_05
  swing1LimitDegrees=45
  swing1Motion=Free
  swing2LimitDegrees=45
  swing2Motion=Free
  twistLimitDegrees=45
  twistMotion=Free
 }
 {
  bone1Name=dyn_cape_l_MSTR
  bone2Name=spine_05
  swing1LimitDegrees=45
  swing1Motion=Free
  swing2LimitDegrees=45
  swing2Motion=Free
  twistLimitDegrees=45
  twistMotion=Free
 }
 {
  bone1Name=spine_04
  bone2Name=spine_03
  swing1LimitDegrees=45
  swing1Motion=Free
  swing2LimitDegrees=45
  swing2Motion=Free
  twistLimitDegrees=45
  twistMotion=Free
 }
 {bone1Name=dyn_cape_mid_03, bone2Name=dyn_cape_mid_02, swing1LimitDegrees=45, swing2LimitDegrees=20, twistLimitDegrees=5}
]
