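# Post-process-style AnimBP for the F_MED_MotorMonth female body, copying the parent pose plus head-mesh jaw and ear bones.
# Without a backbling, cape bones get RigidBody physics and additive offsets for skydive, sprint-jump and idle crouch; wearing one bypasses both.
# A Control Rig finishes either pose.
# The EventGraph sets those alphas, gravity and acceleration overrides, and cape WPO material parameters including skydive wind; the crouch-move offset chain is unconnected.
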AnimBlueprint /BRCosmetics/Characters/Player/Female/Medium/Bodies/F_MED_MotorMonth/Meshes/F_MED_MotorMonth_AnimBP
target_skeleton=&/F_MED_MotorMonth_Skeleton
graphs [
 AnimGraph {nodes [
   n1 = Get CrouchMove() -> (CrouchMove:double)
   n2 = "Transform (Modify) Bone - Bone: dyn_cape_l_MSTR"(Alpha=@n1, Translation="0,0,0", Rotation="0,0,0", Scale="1,1,1") {
    rotationMode=BMM_Additive
    rotationSpace=BCS_BoneSpace
    translationMode=BMM_Additive
    translationSpace=BCS_BoneSpace
    alphaBoolBlend {blendOption=Linear}
    boneToModify {boneName=dyn_cape_l_MSTR}
   }
   n3 = "Transform (Modify) Bone - Bone: dyn_cape_mid_MSTR"(ComponentPose=@n2, Alpha=@n1, Translation="0,1,0", Rotation="0,0,0", Scale="1,1,1") {
    rotationMode=BMM_Additive
    rotationSpace=BCS_BoneSpace
    translationMode=BMM_Additive
    translationSpace=BCS_BoneSpace
    alphaBoolBlend {blendOption=Linear}
    boneToModify {boneName=dyn_cape_mid_MSTR}
    translation {y=1}
   }
   "Transform (Modify) Bone - Bone: dyn_cape_r_MSTR"(ComponentPose=@n3, Alpha=@n1, Translation="0,0,0", Rotation="0,0,0", Scale="1,1,1") {
    rotationMode=BMM_Additive
    rotationSpace=BCS_BoneSpace
    translationMode=BMM_Additive
    translationSpace=BCS_BoneSpace
    alphaBoolBlend {blendOption=Linear}
    boneToModify {boneName=dyn_cape_r_MSTR}
   }
   n4 = Layered blend per bone(
    BasePose=Copy Pose From Mesh() {bUseAttachedParent=true}
    BlendPoses_0=Copy Pose From Mesh(SourceMeshComponent=Get head_mesh() -> (head_mesh:Skeletal Mesh Component Object Reference))
   ) {layerSetup [{branchFilters [
       {boneName=FACIAL_C_UnderChin}
       {boneName=FACIAL_L_UnderChin}
       {boneName=FACIAL_R_UnderChin}
       {boneName=FACIAL_R_Jawline}
       {boneName=FACIAL_C_Jawline}
       {boneName=FACIAL_L_Jawline}
       {boneName=FACIAL_C_FacialRoot}
       {boneName=FACIAL_C_Jaw}
       {boneName=FACIAL_C_Chin}
       {boneName=FACIAL_L_ChinSide}
       {boneName=FACIAL_R_ChinSide}
       {boneName=FACIAL_L_CheekLower}
       {boneName=FACIAL_L_Ear}
       {boneName=FACIAL_L_Masseter}
       {boneName=FACIAL_R_Ear}
       {boneName=FACIAL_R_CheekLower}
       {boneName=FACIAL_R_Masseter}
      ]}]}
   n5 = RigidBody!(
    ComponentPose=Local To Component(LocalPose=@n4)
    Alpha
    OverrideWorldGravity:Vector=Get GravityOverride()
    ExternalForce
    ComponentLinearAccScale:Vector=Get AccScaleOverride() -> (AccScaleOverride:Vector)
   ) -> (Pose) {
    bOverrideWorldGravity=true
    overridePhysicsAsset=&/F_MED_MotorMonth_Physics
    simulationSpace=BaseBoneSpace
    alphaBoolBlend {blendOption=Linear}
    baseBoneRef {boneName=pelvis}
    componentAppliedLinearAccClamp=<1500 1500 1500>
    componentLinearAccScale=<0.35 0.35 0.35>
    componentLinearVelScale=<0.35 0.35 0.35>
   }
   n6 = Get SkydiveFloat() -> (SkydiveFloat:double)
   n7 = "Transform (Modify) Bone - Bone: dyn_cape_l_MSTR"(ComponentPose=@n5, Alpha=@n6, Translation="0,-3,0", Rotation="20,-25,0", Scale="1,1,1") {
    rotationMode=BMM_Additive
    rotationSpace=BCS_BoneSpace
    translationMode=BMM_Additive
    translationSpace=BCS_BoneSpace
    alphaBoolBlend {blendOption=Linear}
    boneToModify {boneName=dyn_cape_l_MSTR}
    rotation {pitch=20, yaw=-25}
    translation {y=-3}
   }
   n8 = "Transform (Modify) Bone - Bone: dyn_cape_mid_MSTR"(ComponentPose=@n7, Alpha=@n6, Translation="0,-3,0", Rotation="0,0,0", Scale="1,1,1") {
    rotationMode=BMM_Additive
    rotationSpace=BCS_BoneSpace
    translationMode=BMM_Additive
    translationSpace=BCS_BoneSpace
    alphaBoolBlend {blendOption=Linear}
    boneToModify {boneName=dyn_cape_mid_MSTR}
    translation {y=-3}
   }
   n9 = "Transform (Modify) Bone - Bone: dyn_cape_r_MSTR"(ComponentPose=@n8, Alpha=@n6, Translation="0,-3,0", Rotation="0,-20,0", Scale="1,1,1") {
    rotationMode=BMM_Additive
    rotationSpace=BCS_BoneSpace
    translationMode=BMM_Additive
    translationSpace=BCS_BoneSpace
    alphaBoolBlend {blendOption=Linear}
    boneToModify {boneName=dyn_cape_r_MSTR}
    rotation {yaw=-20}
    translation {y=-3}
   }
   n10 = Get SkydiveDive() -> (SkydiveDive:double)
   n11 = "Transform (Modify) Bone - Bone: dyn_cape_l_MSTR"(ComponentPose=@n9, Alpha=@n10, Translation="0,-2,0", Rotation="0,0,0", Scale="1,1,1") {
    rotationMode=BMM_Additive
    rotationSpace=BCS_BoneSpace
    translationMode=BMM_Additive
    translationSpace=BCS_BoneSpace
    alphaBoolBlend {blendOption=Linear}
    boneToModify {boneName=dyn_cape_l_MSTR}
    translation {y=-2}
   }
   n12 = "Transform (Modify) Bone - Bone: dyn_cape_mid_MSTR"(ComponentPose=@n11, Alpha=@n10, Translation="0,-2,0", Rotation="0,0,0", Scale="1,1,1") {
    rotationMode=BMM_Additive
    rotationSpace=BCS_BoneSpace
    translationMode=BMM_Additive
    translationSpace=BCS_BoneSpace
    alphaBoolBlend {blendOption=Linear}
    boneToModify {boneName=dyn_cape_mid_MSTR}
    translation {y=-2}
   }
   n13 = "Transform (Modify) Bone - Bone: dyn_cape_r_MSTR"(ComponentPose=@n12, Alpha=@n10, Translation="0,-2,0", Rotation="0,0,0", Scale="1,1,1") {
    rotationMode=BMM_Additive
    rotationSpace=BCS_BoneSpace
    translationMode=BMM_Additive
    translationSpace=BCS_BoneSpace
    alphaBoolBlend {blendOption=Linear}
    boneToModify {boneName=dyn_cape_r_MSTR}
    translation {y=-2}
   }
   n14 = Get TactSprint() -> (TactSprint:double)
   n15 = "Transform (Modify) Bone - Bone: dyn_cape_l_MSTR"(ComponentPose=@n13, Alpha=@n14, Translation="-2,-3,0", Rotation="0,0,0", Scale="1,1,1") {
    rotationMode=BMM_Additive
    rotationSpace=BCS_BoneSpace
    translationMode=BMM_Additive
    translationSpace=BCS_BoneSpace
    alphaBoolBlend {blendOption=Linear}
    boneToModify {boneName=dyn_cape_l_MSTR}
    translation=<-2 -3>
   }
   n16 = "Transform (Modify) Bone - Bone: dyn_cape_mid_MSTR"(ComponentPose=@n15, Alpha=@n14, Translation="-2,-3,0", Rotation="0,0,0", Scale="1,1,1") {
    rotationMode=BMM_Additive
    rotationSpace=BCS_BoneSpace
    translationMode=BMM_Additive
    translationSpace=BCS_BoneSpace
    alphaBoolBlend {blendOption=Linear}
    boneToModify {boneName=dyn_cape_mid_MSTR}
    translation=<-2 -3>
   }
   n17 = "Transform (Modify) Bone - Bone: dyn_cape_r_MSTR"(ComponentPose=@n16, Alpha=@n14, Translation="-2,-3,0", Rotation="0,0,0", Scale="1,1,1") {
    rotationMode=BMM_Additive
    rotationSpace=BCS_BoneSpace
    translationMode=BMM_Additive
    translationSpace=BCS_BoneSpace
    alphaBoolBlend {blendOption=Linear}
    boneToModify {boneName=dyn_cape_r_MSTR}
    translation=<-2 -3>
   }
   n18 = Get Crouching() -> (Crouching:double)
   n19 = "Transform (Modify) Bone - Bone: dyn_cape_l_MSTR"(ComponentPose=@n17, Alpha=@n18, Translation="0,0.5,0", Rotation="0,0,0", Scale="1,1,1") {
    rotationMode=BMM_Additive
    rotationSpace=BCS_BoneSpace
    translationMode=BMM_Additive
    translationSpace=BCS_BoneSpace
    alphaBoolBlend {blendOption=Linear}
    boneToModify {boneName=dyn_cape_l_MSTR}
    translation {y=0.5}
   }
   n20 = "Transform (Modify) Bone - Bone: dyn_cape_mid_MSTR"(ComponentPose=@n19, Alpha=@n18, Translation="0,0.5,0", Rotation="0,0,0", Scale="1,1,1") {
    rotationMode=BMM_Additive
    rotationSpace=BCS_BoneSpace
    translationMode=BMM_Additive
    translationSpace=BCS_BoneSpace
    alphaBoolBlend {blendOption=Linear}
    boneToModify {boneName=dyn_cape_mid_MSTR}
    translation {y=0.5}
   }
   n21 = "Transform (Modify) Bone - Bone: dyn_cape_r_MSTR"(ComponentPose=@n20, Alpha=@n18, Translation="0,0.5,0", Rotation="0,0,0", Scale="1,1,1") {
    rotationMode=BMM_Additive
    rotationSpace=BCS_BoneSpace
    translationMode=BMM_Additive
    translationSpace=BCS_BoneSpace
    alphaBoolBlend {blendOption=Linear}
    boneToModify {boneName=dyn_cape_r_MSTR}
    translation {y=0.5}
   }
   n22 = "Transform (Modify) Bone - Bone: dyn_cape_l_MSTR"(ComponentPose=@n21, Alpha=1, Translation="0,-1,0", Rotation="0,0,0", Scale="1,1,1") {
    translationMode=BMM_Additive
    translationSpace=BCS_BoneSpace
    alphaBoolBlend {blendOption=Linear}
    boneToModify {boneName=dyn_cape_l_MSTR}
    translation {y=-1}
   }
   n23 = "Transform (Modify) Bone - Bone: dyn_cape_mid_MSTR"(ComponentPose=@n22, Alpha=1, Translation="0,-1,0", Rotation="0,0,0", Scale="1,1,1") {
    translationMode=BMM_Additive
    translationSpace=BCS_BoneSpace
    alphaBoolBlend {blendOption=Linear}
    boneToModify {boneName=dyn_cape_mid_MSTR}
    translation {y=-1}
   }
   n24 = "Transform (Modify) Bone - Bone: dyn_cape_r_MSTR"(ComponentPose=@n23, Alpha=1, Translation="0,-1,0", Rotation="0,0,0", Scale="1,1,1") {
    translationMode=BMM_Additive
    translationSpace=BCS_BoneSpace
    alphaBoolBlend {blendOption=Linear}
    boneToModify {boneName=dyn_cape_r_MSTR}
    translation {y=-1}
   }
   n25 = "Transform (Modify) Bone - Bone: dyn_cape_mid_03"(ComponentPose=@n24, Alpha=1, Translation="0,0,0", Rotation="-5,0,0", Scale="1,1,1") {
    translationMode=BMM_Additive
    translationSpace=BCS_BoneSpace
    alphaBoolBlend {blendOption=Linear}
    boneToModify {boneName=dyn_cape_mid_03}
    rotation {pitch=-5}
   }
   n26 = Layered blend per bone(BasePose=Copy Pose From Mesh() {bUseAttachedParent=true}, BlendPoses_0=Copy Pose From Mesh(SourceMeshComponent=Get head_mesh())) {layerSetup [{branchFilters [
       {boneName=FACIAL_C_UnderChin}
       {boneName=FACIAL_L_UnderChin}
       {boneName=FACIAL_R_UnderChin}
       {boneName=FACIAL_R_Jawline}
       {boneName=FACIAL_C_Jawline}
       {boneName=FACIAL_L_Jawline}
       {boneName=FACIAL_C_FacialRoot}
       {boneName=FACIAL_C_Jaw}
       {boneName=FACIAL_C_Chin}
       {boneName=FACIAL_L_ChinSide}
       {boneName=FACIAL_R_ChinSide}
       {boneName=FACIAL_L_CheekLower}
       {boneName=FACIAL_L_Ear}
       {boneName=FACIAL_L_Masseter}
       {boneName=FACIAL_R_Ear}
       {boneName=FACIAL_R_CheekLower}
       {boneName=FACIAL_R_Masseter}
      ]}]}
   n27 = Control Rig(Source=Blend Poses by bool(BlendPose_0=@n26, BlendPose_1=Component To Local(ComponentPose=@n25), bActiveValue=Get isWearBackbling() -> (isWearBackbling:bool)) {blendType=Linear}) {
    alphaBoolBlend {blendOption=Linear}
    controlRigAssetReference {blueprintRigClass=&/F_MED_MotorMonth_CtrlRig.F_MED_MotorMonth_CtrlRig_C}
   }
   Output Pose(Result=@n27) {layerGroup=None}
  ]}
 EventGraph {nodes [
   Event BlueprintUpdateAnimation()
   > n1 = Sequence!(execute) -> (then_0, then_1, then_2:Exec, then_3:Exec, then_4:Exec)
   > n2 = Gravity Override State Function DEPRECATED(CharacterPartAnimInstance=Self-Reference(), GameTimeInSeconds=GetGameTimeInSeconds(), GravityOverrideParams=Get Gravity Override Params())
   > Set GravityOverride(GravityOverride=@n2)
   n3 = Branch(execute=@n1.then_1, Condition=Get bIsCrouching())
   > n4 = Set AccScaleOverride(AccScaleOverride:Vector="0.1,0.1,0.05") -> (Output_Get:Vector)
   n5 = Branch(execute=[@n3.else, @n4.then], Condition=Get isWearBackbling())
   > Set AccScaleOverride(AccScaleOverride="0, 0, 0")
   Set AccScaleOverride(execute=@n5.else, AccScaleOverride="0.3,0.3,0.3")
   Is Valid(exec=@n1.then_2, InputObject=Get PlayerPawn() -> (PlayerPawn:Fort Player Pawn Athena Object Reference))
   > n6 = Branch(Condition=Get bIsInFrontEnd())
   Set WPO Globals Function(
    execute=@n6.else
    Material=Get Mat Instance Dynamic()
    Overlay Material=Get Overlay Mat Instance Dynamic()
    Skeletal Mesh=Get SkelMesh()
    "Local Noise Transform  Location"=Get Local Noise Transform Location() -> (Local Noise Transform Location:Vector)
    Local Noise Offset=Get Local Noise Offset()
   )
   > Set WPO Regional Function!(
    execute
    Material:Material Instance Dynamic Object Reference=Get Mat Instance Dynamic()
    Overlay Material:Material Instance Dynamic Object Reference=Get Overlay Mat Instance Dynamic()
    Skeletal Mesh:Skeletal Mesh Component Object Reference=Get SkelMesh()
    Velocity Transform:Name=cape
    Velocity Interp Speed:double=Get Velocity Interp Speed()
    Collision Transform:Name=cape
    Collision Transform Axis:EAxis Enum
    Invert Collision Transform Axis:bool=false
    Delta Time X:double=0
    Previous Velocity Transform Location:"Vector (by ref)"=Get R1 Velocity Transform Location()
    Previous Velocity Transform Velocity:"Vector (by ref)"=Get R1 Velocity Transform Velocity()
    Region:int=1
   ) -> (then)
   > n7 = Branch(Condition=Get bIsSkydiveFloating())
   > n8 = Set WPO Pho Wind Intensity(WPO Pho Wind Intensity=1)
   SetScalarParameterValue(execute=[@n9.then, @n8.then], self=Get Mat Instance Dynamic(), ParameterName=ClothSim_WindStrength, Value=Get WPO Pho Wind Intensity())
   n9 = Set WPO Pho Wind Intensity(execute=@n7.else, WPO Pho Wind Intensity=0)
   Set SkydiveFloat(execute=@n1.then_3, SkydiveFloat:double=SelectFloat(A=1, B=0, bPickA=Get bIsSkydiveFloating())) -> (Output_Get:double)
   > Set SkydiveDive(SkydiveDive:double=SelectFloat(A=1, B=0, bPickA=Get bIsSkyDiving())) -> (Output_Get:double)
   n10 = OR Boolean(A=Get bIsJumping(), B=Get bIsFalling())
   n11 = Get bIsTacticalSprinting()
   > Set TactSprint(TactSprint:double=SelectFloat(A=1, B=0, bPickA=AND Boolean(A=@n11, B=@n10))) -> (Output_Get:double)
   n12 = NOT Boolean(A=Get bIsMoving2D())
   n13 = Get bIsCrouching()
   > Set Crouching(Crouching:double=SelectFloat(A=1, B=0, bPickA=AND Boolean(A=@n13, B=@n12))) -> (Output_Get:double)
   n14 = Get bIsCrouching()
   n15 = Get bIsMoving2D()
   > Set CrouchMove(CrouchMove:double=SelectFloat(A=1, B=0, bPickA=AND Boolean(A=@n15, B=@n14))) -> (Output_Get:double)
   n16 = Cast To PlayerPawn_Athena(execute=@n1.then_4, Object=TryGetPawnOwner())
   > Set head_mesh(head_mesh:Skeletal Mesh Component Object Reference=GetSkeletalMeshForPartType(self=@n16, PartType=Head)) -> (Output_Get:Skeletal Mesh Component Object Reference)
   Event OnCharacterCustomizationLoaded()
   > n17 = Set SkelMesh(SkelMesh=GetPartSkeletalMeshComponent())
   > n18 = Get Dynamic Material Instance(SkeletalMeshComponent=@n17)
   > Set Mat Instance Dynamic(Mat Instance Dynamic=@n18)
   > n19 = GetOrCreateDynamicOverlayMaterial(MeshComponent=@n17, MaterialName=None)
   > Set Overlay Mat Instance Dynamic(Overlay Mat Instance Dynamic=@n19)
   n20 = CheckForBackbling()
   > Set isWearBackbling(isWearBackbling:bool=@n20) -> (Output_Get:bool)
   Event BlueprintInitializeAnimation()
   > Set SkelMesh(SkelMesh=GetOwnerSkeletalMeshComponent())
   n21 = TryGetPawnOwner()
   > Is Valid(InputObject=@n21)
   > n22 = Cast To FortPlayerPawnAthena(Object=@n21)
   > Set PlayerPawn(PlayerPawn:Fort Player Pawn Athena Object Reference=@n22) -> (Output_Get:Fort Player Pawn Athena Object Reference)
   Set WPO Regional Function(
    Material=Get Mat Instance Dynamic()
    Overlay Material=Get Overlay Mat Instance Dynamic()
    Skeletal Mesh=Get SkelMesh()
    Velocity Transform=foot_l
    Velocity Interp Speed=Get Velocity Interp Speed()
    Collision Transform=calf_l
    Invert Collision Transform Axis=true
    Delta Time X=0
    Previous Velocity Transform Location=Get R2 Velocity Transform Location()
    Previous Velocity Transform Velocity=Get R2 Velocity Transform Velocity()
    Region=2
   )
   Set isWearCape(isWearCape:bool=@n20.bIsBackblingLong) -> (Output_Get:bool)
  ]}
]
variables [
 AccScaleOverride {type=Vector, default=<0.4 0.4 0.4>}
 Body_Mesh {type=SkeletalMeshComponent}
 Crouching {type=number}
 CrouchMove {type=number}
 Gravity Override Params {type=GravityOverrideParamsStruct}
 GravityOverride {type=Vector}
 head_mesh {type=SkeletalMeshComponent}
 isWearBackbling {type=boolean}
 isWearCape {type=boolean}
 Local Noise Offset {type=Vector}
 Local Noise Transform Location {type=Vector}
 Mat Instance Dynamic {type=MaterialInstanceDynamic}
 Overlay Mat Instance Dynamic {type=MaterialInstanceDynamic}
 PlayerPawn {type=FortPlayerPawnAthena}
 R1 Velocity Transform Location {type=Vector}
 R1 Velocity Transform Velocity {type=Vector}
 R2 Velocity Transform Location {type=Vector}
 R2 Velocity Transform Velocity {type=Vector}
 SkelMesh {type=SkeletalMeshComponent}
 SkydiveDive {type=number}
 SkydiveFloat {type=number}
 TactSprint {type=number}
 Velocity Interp Speed {type=number, default=50}
 WPO Pho Wind Intensity {type=number}
]
