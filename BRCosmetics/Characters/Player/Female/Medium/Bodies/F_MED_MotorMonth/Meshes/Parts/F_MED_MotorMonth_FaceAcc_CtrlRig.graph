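# Control Rig for the F_MED_MotorMonth face accessory that applies corrective offsets to hair bones based on head and right-arm pose.
# Spherical pose readers on head and upperarm_r write alpha variables, some multiplied into combined look-and-aim weights.
# These alphas weight Modify Transforms nodes on dyn_hair bones, with a Crouch branch choosing between two arm-forward corrections.
# Several computed alphas, including look left/right, arm down, skydiving and look-down-with-arm-back, drive nothing here.
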ControlRigBlueprint /BRCosmetics/Characters/Player/Female/Medium/Bodies/F_MED_MotorMonth/Meshes/Parts/F_MED_MotorMonth_FaceAcc_CtrlRig
graphs [
 ContainedGraph {nodes [
   n1 = RigVMFunctionEntryNode()
   SequenceExecution_3 = &RigVMFunction_Sequence(ExecuteContext=@n1.ExecuteContext)
   SequenceExecution_3_1 = &RigVMFunction_Sequence(ExecuteContext=@SequenceExecution_3.B)
   SequenceExecution_3_2 = &RigVMFunction_Sequence(ExecuteContext=@SequenceExecution_3_1.B)
   RigVMFunctionReturnNode(A=@SequenceExecution_3.A, B=@SequenceExecution_3_1.A, C=@SequenceExecution_3_2.A, D=@SequenceExecution_3_2.B)
  ]}
 ContainedGraph {nodes [
   n1 = RigVMFunctionEntryNode()
   SequenceExecution_4 = &RigVMFunction_Sequence(ExecuteContext=@n1.ExecuteContext)
   SequenceExecution_4_1 = &RigVMFunction_Sequence(ExecuteContext=@SequenceExecution_4.B)
   SequenceExecution_4_2 = &RigVMFunction_Sequence(ExecuteContext=@SequenceExecution_4_1.B)
   RigVMFunctionReturnNode(A=@SequenceExecution_4.A, B=@SequenceExecution_4_1.A, C=@SequenceExecution_4_2.A, D=@SequenceExecution_4_2.B)
  ]}
 ContainedGraph {nodes [
   n1 = RigVMFunctionEntryNode()
   SequenceExecution_3 = &RigVMFunction_Sequence(ExecuteContext=@n1.ExecuteContext)
   SequenceExecution_3_1 = &RigVMFunction_Sequence(ExecuteContext=@SequenceExecution_3.B)
   SequenceExecution_3_2 = &RigVMFunction_Sequence(ExecuteContext=@SequenceExecution_3_1.B)
   RigVMFunctionReturnNode(A=@SequenceExecution_3.A, B=@SequenceExecution_3_1.A, C=@SequenceExecution_3_2.A, D=@SequenceExecution_3_2.B)
  ]}
 ContainedGraph {nodes [
   n1 = RigVMFunctionEntryNode()
   SequenceExecution_2 = &RigVMFunction_Sequence(ExecuteContext=@n1.ExecuteContext)
   SequenceExecution_2_1 = &RigVMFunction_Sequence(ExecuteContext=@SequenceExecution_2.B)
   SequenceExecution_2_2 = &RigVMFunction_Sequence(ExecuteContext=@SequenceExecution_2_1.B)
   RigVMFunctionReturnNode(A=@SequenceExecution_2.A, B=@SequenceExecution_2_1.A, C=@SequenceExecution_2_2.A, D=@SequenceExecution_2_2.B)
  ]}
 ContainedGraph {nodes [
   n1 = RigVMFunctionEntryNode()
   SequenceExecution_7 = &RigVMFunction_Sequence(ExecuteContext=@n1.ExecuteContext)
   SequenceExecution_7_1 = &RigVMFunction_Sequence(ExecuteContext=@SequenceExecution_7.B)
   SequenceExecution_7_2 = &RigVMFunction_Sequence(ExecuteContext=@SequenceExecution_7_1.B)
   SequenceExecution_7_3 = &RigVMFunction_Sequence(ExecuteContext=@SequenceExecution_7_2.B)
   SequenceExecution_7_4 = &RigVMFunction_Sequence(ExecuteContext=@SequenceExecution_7_3.B)
   RigVMFunctionReturnNode(
    A=@SequenceExecution_7.A
    B=@SequenceExecution_7_1.A
    C=@SequenceExecution_7_2.A
    D=@SequenceExecution_7_3.A
    E=@SequenceExecution_7_4.A
    F=@SequenceExecution_7_4.B
   )
  ]}
 RigVMFunctionLibrary {links [], nodes []}
 RigVMModel {nodes [
   RigUnit_BeginExecution = &RigUnit_BeginExecution()
   SequenceExecution_3_1 = RigVMAggregateNode(ExecuteContext=@RigUnit_BeginExecution.ExecutePin)
   SequenceExecution_3 = RigVMAggregateNode(ExecuteContext=@SequenceExecution_3_1.A)
   SequenceExecution_4 = RigVMAggregateNode(ExecuteContext=@SequenceExecution_3.A)
   SequenceExecution_7 = RigVMAggregateNode(ExecuteContext=@SequenceExecution_4.A)
   n1 = &RigUnit_SphericalPoseReader(
    ActiveRegionScaleFactors=(PositiveWidth=1,NegativeWidth=1,PositiveHeight=1,NegativeHeight=1)
    ActiveRegionSize=0.065
    Debug=(bDrawDebug=false,DebugScale=25,DebugSegments=20,DebugThickness=0.25)
    DriverAxis=(X=-1,Y=0,Z=0)
    DriverItem=(Type=Bone,Name="upperarm_r")
    FalloffRegionScaleFactors=(PositiveWidth=1,NegativeWidth=1,PositiveHeight=1,NegativeHeight=1)
    FalloffSize=0.125
    OptionalParentItem=(Type=Bone)
    RotationOffset=(X=60,Y=0,Z=65)
    ExecutePin=@SequenceExecution_7.A
   )
   > RigVMVariableNode(Value=@n1.OutputParam) {variable=upperarm_r_SPR_back_alpha}
   n2 = &RigUnit_SphericalPoseReader(
    ActiveRegionScaleFactors=(PositiveWidth=1,NegativeWidth=1,PositiveHeight=1,NegativeHeight=1)
    ActiveRegionSize=0.04
    Debug=(bDrawDebug=false,DebugScale=25,DebugSegments=20,DebugThickness=0.25)
    DriverAxis=(X=-1,Y=0,Z=0)
    DriverItem=(Type=Bone,Name="upperarm_r")
    FalloffRegionScaleFactors=(PositiveWidth=1,NegativeWidth=1,PositiveHeight=1,NegativeHeight=1)
    FalloffSize=0.2
    OptionalParentItem=(Type=Bone)
    RotationOffset=(X=-60,Y=15,Z=0)
    ExecutePin=@SequenceExecution_7.F
   )
   > RigVMVariableNode(Value=@n2.OutputParam) {variable=upperarm_r_down_alpha}
   n3 = &RigUnit_SphericalPoseReader(
    ActiveRegionScaleFactors=(PositiveWidth=1,NegativeWidth=1,PositiveHeight=1,NegativeHeight=1)
    ActiveRegionSize=0.075
    Debug=(bDrawDebug=false,DebugScale=25,DebugSegments=20,DebugThickness=0.25)
    DriverAxis=(X=-1,Y=0,Z=0)
    DriverItem=(Type=Bone,Name="upperarm_r")
    FalloffRegionScaleFactors=(PositiveWidth=1,NegativeWidth=1,PositiveHeight=1,NegativeHeight=1)
    FalloffSize=0.2
    OptionalParentItem=(Type=Bone)
    RotationOffset=(X=-55,Y=0,Z=-20)
    ExecutePin=@SequenceExecution_7.D
   )
   > RigVMVariableNode(Value=@n3.OutputParam) {variable=upperarm_r_rifle_up_alpha}
   n4 = &RigUnit_SphericalPoseReader(
    ActiveRegionScaleFactors=(PositiveWidth=1,NegativeWidth=1,PositiveHeight=1,NegativeHeight=1)
    ActiveRegionSize=0.04
    Debug=(bDrawDebug=false,DebugScale=25,DebugSegments=20,DebugThickness=0.25)
    DriverAxis=(X=-1,Y=0,Z=0)
    DriverItem=(Type=Bone,Name="upperarm_r")
    FalloffRegionScaleFactors=(PositiveWidth=1,NegativeWidth=1,PositiveHeight=1,NegativeHeight=1)
    FalloffSize=0.2
    OptionalParentItem=(Type=Bone)
    RotationOffset=(X=-115,Y=0,Z=25)
    ExecutePin=@SequenceExecution_7.B
   )
   > RigVMVariableNode(Value=@n4.OutputParam) {variable=uppearm_r_SPR_aim_up_two_hands_alpha}
   n5 = &RigUnit_SphericalPoseReader(
    ActiveRegionScaleFactors=(PositiveWidth=2,NegativeWidth=2,PositiveHeight=1,NegativeHeight=1)
    ActiveRegionSize=0.05
    Debug=(bDrawDebug=false,DebugScale=25,DebugSegments=20,DebugThickness=0.25)
    DriverAxis=(X=-1,Y=0,Z=0)
    DriverItem=(Type=Bone,Name="upperarm_r")
    FalloffRegionScaleFactors=(PositiveWidth=1,NegativeWidth=1,PositiveHeight=0.75,NegativeHeight=0.75)
    FalloffSize=0.2
    FlipWidthScaling=false
    OptionalParentItem=(Type=Bone)
    RotationOffset=(X=15,Y=150,Z=0)
    ExecutePin=@SequenceExecution_7.C
   )
   > RigVMVariableNode(Value=@n5.OutputParam) {variable=upperarm_r_skydiving}
   n6 = &RigUnit_SphericalPoseReader(
    ActiveRegionScaleFactors=(PositiveWidth=1,NegativeWidth=1,PositiveHeight=1,NegativeHeight=1)
    ActiveRegionSize=0.05
    Debug=(bDrawDebug=false,DebugScale=25,DebugSegments=20,DebugThickness=0.25)
    DriverAxis=(X=-1,Y=0,Z=0)
    DriverItem=(Type=Bone,Name="upperarm_r")
    FalloffRegionScaleFactors=(PositiveWidth=1,NegativeWidth=1,PositiveHeight=1,NegativeHeight=1)
    FalloffSize=0.2
    OptionalParentItem=(Type=Bone)
    RotationOffset=(X=-90,Y=0,Z=0)
    ExecutePin=@SequenceExecution_7.E
   )
   > RigVMVariableNode(Value=@n6.OutputParam) {variable=upperarm_r_forward_alpha}
   SequenceExecution_2 = RigVMAggregateNode(ExecuteContext=@SequenceExecution_4.B)
   n7 = &RigUnit_SphericalPoseReader(
    ActiveRegionScaleFactors=(PositiveWidth=1,NegativeWidth=0.5,PositiveHeight=1.5,NegativeHeight=1.5)
    ActiveRegionSize=0.1
    Debug=(bDrawDebug=false,DebugScale=25,DebugSegments=20,DebugThickness=0.25)
    DriverAxis=(X=0,Y=1,Z=0)
    DriverItem=(Type=Bone,Name="head")
    FalloffRegionScaleFactors=(PositiveWidth=1,NegativeWidth=0.5,PositiveHeight=1.5,NegativeHeight=1.5)
    FalloffSize=0.25
    FlipWidthScaling=false
    OptionalParentItem=(Type=Bone)
    RotationOffset=(X=90,Y=0,Z=-25)
    ExecutePin=@SequenceExecution_2.A
   )
   > RigVMVariableNode(Value=@n7.OutputParam) {variable=head_SPR_look_up_alpha}
   n8 = &RigUnit_SphericalPoseReader(
    ActiveRegionScaleFactors=(PositiveWidth=2,NegativeWidth=0.5,PositiveHeight=2,NegativeHeight=2)
    ActiveRegionSize=0.075
    Debug=(bDrawDebug=false,DebugScale=25,DebugSegments=20,DebugThickness=0.25)
    DriverAxis=(X=0,Y=1,Z=0)
    DriverItem=(Type=Bone,Name="head")
    FalloffRegionScaleFactors=(PositiveWidth=1,NegativeWidth=0.4,PositiveHeight=1,NegativeHeight=1)
    FalloffSize=0.2
    FlipHeightScaling=false
    FlipWidthScaling=true
    OptionalParentItem=(Type=Bone)
    RotationOffset=(X=90,Y=0,Z=12.5)
    ExecutePin=@SequenceExecution_2.B
   )
   > RigVMVariableNode(Value=@n8.OutputParam) {variable=head_SPR_look_down_alpha}
   n9 = &RigUnit_SphericalPoseReader(
    ActiveRegionScaleFactors=(PositiveWidth=0.75,NegativeWidth=0.75,PositiveHeight=1,NegativeHeight=0.5)
    ActiveRegionSize=0.1
    Debug=(bDrawDebug=false,DebugScale=25,DebugSegments=20,DebugThickness=0.25)
    DriverAxis=(X=0,Y=1,Z=0)
    DriverItem=(Type=Bone,Name="head")
    FalloffRegionScaleFactors=(PositiveWidth=1,NegativeWidth=1,PositiveHeight=1,NegativeHeight=0.5)
    FalloffSize=0.2
    OptionalParentItem=(Type=Bone)
    RotationOffset=(X=120,Y=0,Z=0)
    ExecutePin=@SequenceExecution_2.C
   )
   > RigVMVariableNode(Value=@n9.OutputParam) {variable=head_SPR_look_left_alpha}
   n10 = &RigUnit_SphericalPoseReader(
    ActiveRegionScaleFactors=(PositiveWidth=0.75,NegativeWidth=0.75,PositiveHeight=1,NegativeHeight=0.5)
    ActiveRegionSize=0.1
    Debug=(bDrawDebug=false,DebugScale=25,DebugSegments=20,DebugThickness=0.25)
    DriverAxis=(X=0,Y=1,Z=0)
    DriverItem=(Type=Bone,Name="head")
    FalloffRegionScaleFactors=(PositiveWidth=1,NegativeWidth=1,PositiveHeight=1,NegativeHeight=0.5)
    FalloffSize=0.2
    FlipHeightScaling=true
    FlipWidthScaling=false
    OptionalParentItem=(Type=Bone)
    RotationOffset=(X=60,Y=0,Z=0)
    ExecutePin=@SequenceExecution_2.D
   )
   > RigVMVariableNode(Value=@n10.OutputParam) {variable=head_SPR_look_right_alpha}
   RigVMVariableNode(
    Value=&RigVMFunction_MathFloatMul(B=RigVMVariableNode() {variable=uppearm_r_SPR_aim_up_two_hands_alpha}.Value, A=RigVMVariableNode() {variable=head_SPR_look_up_alpha}.Value) {template="Multiply::Execute(in A,in B,out Result)"}.Result
    ExecuteContext=@SequenceExecution_3.B
   ) {variable=look_up_and_left_two_hands_multiplied_alpha}
   RigVMVariableNode(
    Value=&RigVMFunction_MathFloatMul(A=RigVMVariableNode() {variable=head_SPR_look_up_alpha}.Value, B=RigVMVariableNode() {variable=upperarm_r_rifle_up_alpha}.Value) {template="Multiply::Execute(in A,in B,out Result)"}.Result
    ExecuteContext=@SequenceExecution_3.C
   ) {variable=look_up_two_hands_rifle_multiplied_alpha}
   RigVMVariableNode(
    Value=&RigVMFunction_MathFloatMul(B=RigVMVariableNode() {variable=head_SPR_look_down_alpha}.Value, A=RigVMVariableNode() {variable=upperarm_r_SPR_back_alpha}.Value) {template="Multiply::Execute(in A,in B,out Result)"}.Result
    ExecuteContext=@SequenceExecution_3.D
   ) {variable=look_down_upperarm_r_back_multiplied_alpha}
   Branch_1_1_1_RigVMUnitNode = &RigVMFunction_ControlFlowBranch(ExecuteContext=@SequenceExecution_3_1.D, Condition=RigVMVariableNode() {variable=Crouch}.Value)
   &RigUnit_ModifyTransforms(
    ItemToModify="((Item=(Type=Bone,Name=\"dyn_hair_front_r_03\"),Transform=(Rotation=(X=0,Y=0,Z=0,W=1),Translation=(X=0,Y=-2,Z=0),Scale3D=(X=1,Y=1,Z=1))),(Item=(Type=Bone,Name=\"dyn_hairTips_r_02\"),Transform=(Rotation=(X=0,Y=0,Z=0,W=1),Translation=(X=0,Y=0,Z=0),Scale3D=(X=1,Y ...(truncated)"
    ExecutePin=@Branch_1_1_1_RigVMUnitNode.False
    ItemToModify.1.Transform.Rotation=&RigVMFunction_MathQuaternionFromEuler(Euler=(X=0,Y=-5,Z=10)).Result
    ItemToModify.0.Transform.Rotation=&RigVMFunction_MathQuaternionFromEuler(Euler=(X=0,Y=-20,Z=0)).Result
    Weight=RigVMVariableNode() {variable=upperarm_r_forward_alpha}.Value
   )
   n11 = &RigUnit_SphericalPoseReader(
    ActiveRegionScaleFactors=(PositiveWidth=1,NegativeWidth=1,PositiveHeight=1,NegativeHeight=1)
    ActiveRegionSize=0.05
    Debug=(bDrawDebug=true,DebugScale=25,DebugSegments=20,DebugThickness=0.25)
    DriverAxis=(X=-1,Y=0,Z=0)
    DriverItem=(Type=Bone,Name="upperarm_r")
    FalloffRegionScaleFactors=(PositiveWidth=1,NegativeWidth=1,PositiveHeight=1,NegativeHeight=1)
    FalloffSize=0.2
    OptionalParentItem=(Type=Bone)
    RotationOffset=(X=-90,Y=0,Z=15)
    ExecutePin=@Branch_1_1_1_RigVMUnitNode.True
   )
   > &RigUnit_ModifyTransforms(
    ItemToModify="((Item=(Type=Bone,Name=\"dyn_hair_bk_01\"),Transform=(Rotation=(X=0,Y=0,Z=0,W=1),Translation=(X=0,Y=0,Z=0),Scale3D=(X=1,Y=1,Z=1))),(Item=(Type=Bone,Name=\"dyn_hair_bk_02\"),Transform=(Rotation=(X=0,Y=0,Z=0,W=1),Translation=(X=0,Y=0,Z=0),Scale3D=(X=1,Y=1,Z=1))) ...(truncated)"
    ItemToModify.1.Transform.Rotation=&RigVMFunction_MathQuaternionFromEuler(Euler=(X=0,Y=5,Z=0)).Result
    ItemToModify.2.Transform.Rotation=&RigVMFunction_MathQuaternionFromEuler(Euler=(X=0,Y=-20,Z=0)).Result
    ItemToModify.0.Transform.Rotation=&RigVMFunction_MathQuaternionFromEuler(Euler=(X=0,Y=15,Z=0)).Result
    Weight=@n11.OutputParam
   )
   n12 = RigVMVariableNode() {variable=look_up_and_left_two_hands_multiplied_alpha}
   &RigUnit_ModifyTransforms(
    ItemToModify="((Item=(Type=Bone,Name=\"dyn_hair_r_MSTR\"),Transform=(Rotation=(X=0,Y=0,Z=0,W=1),Translation=(X=-1,Y=0,Z=1),Scale3D=(X=1,Y=1,Z=1))),(Item=(Type=Bone,Name=\"dyn_hair_r_01\"),Transform=(Rotation=(X=0,Y=0,Z=0,W=1),Translation=(X=0,Y=0,Z=0),Scale3D=(X=1,Y=1,Z=1)) ...(truncated)"
    ItemToModify.2.Transform.Rotation=&RigVMFunction_MathQuaternionFromEuler(Euler=(X=-5,Y=0,Z=10)).Result
    ItemToModify.3.Transform.Rotation=&RigVMFunction_MathQuaternionFromEuler(Euler=(X=0,Y=30,Z=0)).Result
    ItemToModify.1.Transform.Rotation=&RigVMFunction_MathQuaternionFromEuler(Euler=(X=0,Y=0,Z=0)).Result
    ItemToModify.0.Transform.Rotation=&RigVMFunction_MathQuaternionFromEuler(Euler=(X=0,Y=-5,Z=0)).Result
    Weight=@n12.Value
    ExecutePin=@SequenceExecution_3_1.B
   )
   > &RigUnit_ModifyTransforms(
    ItemToModify="((Item=(Type=Bone,Name=\"dyn_hair_front_r_01\"),Transform=(Rotation=(X=0,Y=0,Z=0,W=1),Translation=(X=0,Y=1,Z=1),Scale3D=(X=1,Y=1,Z=1))),(Item=(Type=Bone,Name=\"dyn_hair_front_r_02\"),Transform=(Rotation=(X=0,Y=0,Z=0,W=1),Translation=(X=0,Y=0,Z=0),Scale3D=(X=1, ...(truncated)"
    ItemToModify.3.Transform.Rotation=&RigVMFunction_MathQuaternionFromEuler(Euler=(X=0,Y=-15,Z=0)).Result
    ItemToModify.2.Transform.Rotation=&RigVMFunction_MathQuaternionFromEuler(Euler=(X=0,Y=30,Z=0)).Result
    ItemToModify.1.Transform.Rotation=&RigVMFunction_MathQuaternionFromEuler(Euler=(X=0,Y=0,Z=0)).Result
    ItemToModify.0.Transform.Rotation=&RigVMFunction_MathQuaternionFromEuler(Euler=(X=10,Y=-5,Z=10)).Result
    Weight=@n12.Value
   )
   > &RigUnit_ModifyTransforms(
    ItemToModify="((Item=(Type=Bone,Name=\"dyn_hair_bk_MSTR\"),Transform=(Rotation=(X=0,Y=0,Z=0,W=1),Translation=(X=0,Y=0,Z=0),Scale3D=(X=1,Y=1,Z=1))),(Item=(Type=Bone,Name=\"dyn_hairTips_r_02\"),Transform=(Rotation=(X=0,Y=0,Z=0,W=1),Translation=(X=0,Y=0,Z=0),Scale3D=(X=1,Y=1,Z ...(truncated)"
    ItemToModify.1.Transform.Rotation=&RigVMFunction_MathQuaternionFromEuler(Euler=(X=0,Y=20,Z=15)).Result
    ItemToModify.0.Transform.Rotation=&RigVMFunction_MathQuaternionFromEuler(Euler=(X=-5,Y=-5,Z=0)).Result
    Weight=@n12.Value
   )
   n13 = RigVMVariableNode() {variable=look_up_two_hands_rifle_multiplied_alpha}
   > &RigUnit_ModifyTransforms(
    ItemToModify="((Item=(Type=Bone,Name=\"dyn_hair_l_01\"),Transform=(Rotation=(X=0,Y=0,Z=0,W=1),Translation=(X=0,Y=0,Z=0),Scale3D=(X=1,Y=1,Z=1))),(Item=(Type=Bone,Name=\"dyn_hair_l_02\"),Transform=(Rotation=(X=0,Y=0,Z=0,W=1),Translation=(X=0,Y=0,Z=0),Scale3D=(X=1,Y=1,Z=1))),( ...(truncated)"
    ItemToModify.0.Transform.Rotation=&RigVMFunction_MathQuaternionFromEuler(Euler=(X=0,Y=5,Z=0)).Result
    ItemToModify.1.Transform.Rotation=&RigVMFunction_MathQuaternionFromEuler(Euler=(X=0,Y=0,Z=0)).Result
    ItemToModify.2.Transform.Rotation=&RigVMFunction_MathQuaternionFromEuler(Euler=(X=0,Y=0,Z=0)).Result
    ItemToModify.4.Transform.Rotation=&RigVMFunction_MathQuaternionFromEuler(Euler=(X=0,Y=0,Z=0)).Result
    Weight=@n13.Value
   )
   > &RigUnit_ModifyTransforms(
    ItemToModify="((Item=(Type=Bone,Name=\"dyn_hair_r_MSTR\"),Transform=(Rotation=(X=0,Y=0,Z=0,W=1),Translation=(X=0,Y=0,Z=0),Scale3D=(X=1,Y=1,Z=1))),(Item=(Type=Bone,Name=\"dyn_hair_r_02\"),Transform=(Rotation=(X=0,Y=0,Z=0,W=1),Translation=(X=0,Y=0,Z=0),Scale3D=(X=1,Y=1,Z=1))) ...(truncated)"
    ItemToModify.0.Transform.Rotation=&RigVMFunction_MathQuaternionFromEuler(Euler=(X=0,Y=0,Z=0)).Result
    ItemToModify.2.Transform.Rotation=&RigVMFunction_MathQuaternionFromEuler(Euler=(X=0,Y=-20,Z=0)).Result
    ItemToModify.3.Transform.Rotation=&RigVMFunction_MathQuaternionFromEuler(Euler=(X=0,Y=0,Z=10)).Result
    ItemToModify.1.Transform.Rotation=&RigVMFunction_MathQuaternionFromEuler(Euler=(X=0,Y=0,Z=0)).Result
    Weight=@n13.Value
   )
   > &RigUnit_ModifyTransforms(
    ItemToModify="((Item=(Type=Bone,Name=\"dyn_hairTips_r_02\"),Transform=(Rotation=(X=0,Y=0,Z=0,W=1),Translation=(X=0,Y=0,Z=0),Scale3D=(X=1,Y=1,Z=1))),(Item=(Type=Bone,Name=\"dyn_hair_bk_03\"),Transform=(Rotation=(X=0,Y=0,Z=0,W=1),Translation=(X=0,Y=0,Z=1),Scale3D=(X=1,Y=1,Z=1 ...(truncated)"
    ItemToModify.0.Transform.Rotation=&RigVMFunction_MathQuaternionFromEuler(Euler=(X=0,Y=0,Z=-20)).Result
    Weight=RigVMVariableNode() {variable=head_SPR_look_up_alpha}.Value
   )
   &RigUnit_ModifyTransforms(
    ItemToModify="((Item=(Type=Bone,Name=\"dyn_hair_bk_03\"),Transform=(Rotation=(X=0,Y=0,Z=0,W=1),Translation=(X=0,Y=0,Z=2),Scale3D=(X=1,Y=1,Z=1))),(Item=(Type=Bone,Name=\"dyn_hair_r_04\"),Transform=(Rotation=(X=0,Y=0,Z=0,W=1),Translation=(X=0,Y=0,Z=2),Scale3D=(X=1,Y=1,Z=1))), ...(truncated)"
    ItemToModify.1.Transform.Rotation=&RigVMFunction_MathQuaternionFromEuler(Euler=(X=0,Y=0,Z=0)).Result
    ItemToModify.0.Transform.Rotation=&RigVMFunction_MathQuaternionFromEuler(Euler=(X=0,Y=0,Z=0)).Result
    Weight=RigVMVariableNode() {variable=upperarm_r_SPR_back_alpha}.Value
    ExecutePin=@SequenceExecution_3_1.C
   )
   > &RigUnit_ModifyTransforms(
    ItemToModify="((Item=(Type=Bone,Name=\"dyn_hair_bk_03\"),Transform=(Rotation=(X=0,Y=0,Z=0,W=1),Translation=(X=0,Y=0,Z=-2),Scale3D=(X=1,Y=1,Z=1))),(Item=(Type=Bone,Name=\"dyn_hair_r_04\"),Transform=(Rotation=(X=0,Y=0,Z=0,W=1),Translation=(X=0,Y=0,Z=0),Scale3D=(X=1,Y=1,Z=1))) ...(truncated)"
    ItemToModify.1.Transform.Rotation=&RigVMFunction_MathQuaternionFromEuler(Euler=(X=0,Y=0,Z=0)).Result
    ItemToModify.0.Transform.Rotation=&RigVMFunction_MathQuaternionFromEuler(Euler=(X=0,Y=0,Z=0)).Result
    Weight=RigVMVariableNode() {variable=head_SPR_look_down_alpha}.Value
   )
   &RigVMFunction_MathQuaternionFromEuler(Euler=(X=0,Y=-15,Z=0))
   &RigVMFunction_MathQuaternionFromEuler(Euler=(X=0,Y=30,Z=0))
  ]}
]
hierarchy [
 root {loc=<0 0 0>, rot=<0 0 0 1>}
 pelvis {parent=root, loc=<0 -0.5301 98.0571>, rot=<0.0127 0.707 -0.0127 -0.707>}
 spine_01 {parent=pelvis, loc=<6.1208 -0.9328 0>, rot=<0 0 0.034 -0.9994>}
 spine_02 {parent=spine_01, loc=<6.5996 -0.2095 0>, rot=<0 0 0.0238 -0.9997>}
 spine_03 {parent=spine_02, loc=<6.4914 -0.1237 0>, rot=<0 0 0.0149 -0.9999>}
 spine_04 {parent=spine_03, loc=<6.9598 0.0794 0>, rot=<0 0 -0.0132 -0.9999>}
 spine_05 {parent=spine_04, loc=<6.9987 0 0>, rot=<0 0 -0.0015 -1>}
 neck_01 {parent=spine_05, loc=<11.3138 0.1529 0>, rot=<0 0 -0.2003 -0.9797>}
 neck_02 {parent=neck_01, loc=<4.5627 0 0>, rot=<0 0 0.0504 -0.9987>}
 head {parent=neck_02, loc=<5.3914 0 0>, rot=<0 0 0.115 -0.9934>}
 dyn_hair_l_MSTR {parent=head, loc=<8.8541 0.6916 -6.499>, rot=<0.0262 -0.7919 0.5984 -0.1187>}
 dyn_hair_l_00 {parent=dyn_hair_l_MSTR, loc=<0.5 0 0>, rot=<0 0 0 1>}
 dyn_hair_l_01 {parent=dyn_hair_l_00, loc=<7.0495 0 0>, rot=<0.0202 0.1166 0.0179 -0.9928>}
 dyn_hair_l_02 {parent=dyn_hair_l_01, loc=<7.1616 0 0>, rot=<-0.086 0.2331 0.0422 -0.9677>}
 dyn_hair_l_03 {parent=dyn_hair_l_02, loc=<8.7592 0 0>, rot=<-0.0205 -0.2538 0.0767 -0.964>}
 dyn_hair_l_04 {parent=dyn_hair_l_03, loc=<8.993 0 0>, rot=<0.0802 -0.1932 -0.017 -0.9777>}
 dyn_hairTips_l_01 {parent=dyn_hair_l_01, loc=<5.2071 2.7127 0.1538>, rot=<-0.0829 0.2895 0.0083 -0.9535>}
 dyn_hairTips_l_02 {parent=dyn_hairTips_l_01, loc=<5.6989 0 0>, rot=<-0.0381 -0.083 0.0552 -0.9943>}
 dyn_hairTips_l_03 {parent=dyn_hair_l_01, loc=<3.4996 1.2086 -5.1942>, rot=<0.0504 0.0477 -0.15 -0.9862>}
 dyn_hairTips_l_04 {parent=dyn_hairTips_l_03, loc=<5.6989 0 0>, rot=<-0.0244 0.0798 0.0063 -0.9965>}
 dyn_hair_bk_MSTR {parent=head, loc=<11.6835 -5.2363 0>, rot=<0.1215 -0.6966 0.6966 -0.1215>}
 dyn_hair_bk_00 {parent=dyn_hair_bk_MSTR, loc=<0.5 0 0>, rot=<0 -0.0572 0 -0.9984>}
 dyn_hair_bk_01 {parent=dyn_hair_bk_00, loc=<6.0938 0 0>, rot=<0 -0.0272 0 -0.9996>}
 dyn_hair_bk_02 {parent=dyn_hair_bk_01, loc=<8.7854 0 0>, rot=<0 0.1593 0 -0.9872>}
 dyn_hair_bk_03 {parent=dyn_hair_bk_02, loc=<9.4823 0 0>, rot=<0 -0.1683 0 -0.9857>}
 dyn_hair_bk_04 {parent=dyn_hair_bk_03, loc=<10.3536 0 0>, rot=<0 -0.0751 0 -0.9972>}
 dyn_hair_r_MSTR {parent=head, loc=<10.0791 -1.0364 6.0327>, rot=<0.1315 -0.5972 0.7911 0.0152>}
 dyn_hair_r_00 {parent=dyn_hair_r_MSTR, loc=<0.5 0 0>, rot=<-0.0006 0.0016 -0.0049 -1>}
 dyn_hair_r_01 {parent=dyn_hair_r_00, loc=<6.1151 0 0>, rot=<-0.0063 0.0224 -0.0723 -0.9971>}
 dyn_hair_r_02 {parent=dyn_hair_r_01, loc=<7.6839 0 0>, rot=<-0.0459 0.3353 -0.0465 -0.9398>}
 dyn_hair_r_03 {parent=dyn_hair_r_02, loc=<7.9495 0 0>, rot=<0.1523 -0.2716 -0.0403 -0.9494>}
 dyn_hair_r_04 {parent=dyn_hair_r_03, loc=<8.1126 0 0>, rot=<0.0022 -0.1842 -0.0117 -0.9828>}
 dyn_hair_front_r_04 {parent=dyn_hair_r_02, loc=<2.4425 -2.3557 -5.076>, rot=<-0.3065 0.2035 0.2358 -0.8995>}
 dyn_hair_front_r_05 {parent=dyn_hair_front_r_04, loc=<6.0993 0 0>, rot=<0.1117 -0.2236 0.009 -0.9682>}
 dyn_hairTips_r_01 {parent=dyn_hair_r_01, loc=<4.1928 -3.3741 -2.8373>, rot=<-0.0041 0.2712 0.2287 -0.9349>}
 dyn_hair_front_l_MSTR {parent=head, loc=<9.4738 8.1631 -2.5646>, rot=<-0.2909 -0.6419 0.6961 -0.137>}
 dyn_hair_front_l_01 {parent=dyn_hair_front_l_MSTR, loc=<0 0 0>, rot=<0 0 0 1>}
 dyn_hair_front_l_02 {parent=dyn_hair_front_l_01, loc=<3.7785 0.3251 -0.1369>, rot=<0 0.2235 0.3092 -0.9244>}
 dyn_hair_front_l_03 {parent=dyn_hair_front_l_02, loc=<5.5049 0.0172 -0.1106>, rot=<0.13 -0.1593 -0.1677 -0.9642>}
 dyn_hair_front_l_04 {parent=dyn_hair_front_l_03, loc=<5.0265 0 0>, rot=<-0.1152 0 0 -0.9933>}
 dyn_hair_front_r_MSTR {parent=head, loc=<9.1613 8.37 2.9443>, rot=<0.279 -0.2663 0.9195 0.0756>}
 dyn_hair_front_r_01 {parent=dyn_hair_front_r_MSTR, loc=<0 0 0>, rot=<0 0 0 1>}
 dyn_hair_front_r_hairTips_01 {parent=dyn_hair_front_r_01, loc=<5.6501 -0.458 -0.2718>, rot=<0.398 -0.0087 -0.2413 -0.885>}
 dyn_hair_front_r_hairTips_02 {parent=dyn_hair_front_r_hairTips_01, loc=<4.919 0.5766 -0.4511>, rot=<0 0 0 1>}
 dyn_hair_front_r_02 {parent=dyn_hair_front_r_01, loc=<7.8648 1.8813 0.3588>, rot=<0.2645 0.0239 -0.3052 -0.9145>}
 dyn_hairTips_r_04 {parent=dyn_hair_front_r_02, loc=<3.5808 -1.4764 0.2923>, rot=<-0.1271 0.2141 0.2045 -0.9467>}
 dyn_hair_front_r_03 {parent=dyn_hair_front_r_02, loc=<6.9429 0 0>, rot=<-0.0371 0.1834 0.0207 -0.9821>}
 dyn_hairTips_r_02 {parent=dyn_hair_front_r_03, loc=<3.872 0.6037 -1.8638>, rot=<0.0718 -0.1611 0.2226 -0.9588>}
 dyn_hairTips_r_03 {parent=dyn_hairTips_r_02, loc=<4.1672 0 0>, rot=<-0.1241 0.2651 0.0485 -0.955>}
 FACIAL_C_FacialRoot {parent=head, loc=<2.1524 -0.746 -0.0002>, rot=<0.4879 -0.5118 -0.4879 -0.5118>}
 FACIAL_C_Forehead {parent=FACIAL_C_FacialRoot, loc=<-0.0002 -3.1236 8.8011>, rot=<-0.0564 0 0 -0.9984>}
 FACIAL_L_ForeheadIn {parent=FACIAL_C_FacialRoot, loc=<1.5463 -3.0808 8.6236>, rot=<-0.0736 -0.0696 0.0034 -0.9948>}
 FACIAL_R_ForeheadIn {parent=FACIAL_C_FacialRoot, loc=<-1.551 -3.0811 8.6208>, rot=<-0.0736 0.0697 -0.0034 -0.9948>}
 FACIAL_L_ForeheadMid {parent=FACIAL_C_FacialRoot, loc=<2.9337 -3.3708 8.0621>, rot=<-0.0837 -0.1753 -0.0229 -0.9807>}
 FACIAL_R_ForeheadMid {parent=FACIAL_C_FacialRoot, loc=<-2.9413 -3.3703 8.0596>, rot=<-0.0837 0.1754 0.0229 -0.9807>}
 FACIAL_L_ForeheadOut {parent=FACIAL_C_FacialRoot, loc=<4.3183 -3.3091 7.1418>, rot=<-0.1338 -0.2545 -0.0825 -0.9542>}
 FACIAL_R_ForeheadOut {parent=FACIAL_C_FacialRoot, loc=<-4.3243 -3.3073 7.1415>, rot=<-0.1338 0.2546 0.0825 -0.9542>}
 FACIAL_L_EyesackUpper {parent=FACIAL_C_FacialRoot, loc=<2.8043 -2.261 8.6645>, rot=<0.0359 -0.0834 -0.0656 -0.9937>}
 FACIAL_R_EyesackUpper {parent=FACIAL_C_FacialRoot, loc=<-2.8127 -2.2596 8.6629>, rot=<0.0359 0.0834 0.0656 -0.9937>}
 FACIAL_L_EyelidUpperFurrow {parent=FACIAL_C_FacialRoot, loc=<2.6602 -1.8995 8.5581>, rot=<-0.0428 -0.1497 -0.0816 -0.9844>}
 FACIAL_R_EyelidUpperFurrow {parent=FACIAL_C_FacialRoot, loc=<-2.6752 -1.8974 8.5546>, rot=<-0.0428 0.1498 0.0816 -0.9844>}
 FACIAL_L_EyelidUpperA {parent=FACIAL_C_FacialRoot, loc=<2.921 -0.7747 6.9469>, rot=<-0.1607 0 0 -0.987>}
 FACIAL_R_EyelidUpperA {parent=FACIAL_C_FacialRoot, loc=<-2.9221 -0.7719 6.9485>, rot=<-0.1607 0 0 -0.987>}
 FACIAL_L_EyelidLowerA {parent=FACIAL_C_FacialRoot, loc=<2.921 -0.7747 6.9469>, rot=<0.2715 0 0 -0.9624>}
 FACIAL_R_EyelidLowerA {parent=FACIAL_C_FacialRoot, loc=<-2.9221 -0.7719 6.9485>, rot=<0.2715 0 0 -0.9624>}
 FACIAL_L_EyeCornerInner {parent=FACIAL_C_FacialRoot, loc=<1.8003 -0.2733 7.8338>, rot=<-0.013 0 0 -0.9999>}
 FACIAL_R_EyeCornerInner {parent=FACIAL_C_FacialRoot, loc=<-1.805 -0.26 7.8302>, rot=<-0.013 0 0 -0.9999>}
 FACIAL_L_EyeCornerOuter {parent=FACIAL_C_FacialRoot, loc=<4.551 -0.8779 6.9422>, rot=<-0.012 -0.3825 -0.005 -0.9238>}
 FACIAL_R_EyeCornerOuter {parent=FACIAL_C_FacialRoot, loc=<-4.5643 -0.8804 6.9329>, rot=<-0.012 0.3826 0.005 -0.9238>}
 FACIAL_L_EyesackLower {parent=FACIAL_C_FacialRoot, loc=<3.221 0.7999 6.7327>, rot=<-0.0351 -0.1394 -0.0241 -0.9893>}
 FACIAL_R_EyesackLower {parent=FACIAL_C_FacialRoot, loc=<-3.2341 0.8217 6.7248>, rot=<-0.0352 0.1394 0.0241 -0.9893>}
 FACIAL_L_CheekInner {parent=FACIAL_C_FacialRoot, loc=<2.0521 1.2131 7.5653>, rot=<-0.1358 -0.0854 -0.1982 -0.9669>}
 FACIAL_R_CheekInner {parent=FACIAL_C_FacialRoot, loc=<-2.0583 1.2168 7.5682>, rot=<-0.1358 0.0855 0.1982 -0.9669>}
 FACIAL_L_CheekOuter {parent=FACIAL_C_FacialRoot, loc=<4.0237 1.1926 5.9117>, rot=<0.0432 -0.3782 0.0955 -0.9198>}
 FACIAL_R_CheekOuter {parent=FACIAL_C_FacialRoot, loc=<-4.0284 1.1989 5.9119>, rot=<0.0432 0.3782 -0.0954 -0.9198>}
 FACIAL_L_NasolabialBulge {parent=FACIAL_C_FacialRoot, loc=<3.4025 4.3606 5.7401>, rot=<0.0459 -0.3776 0.0189 -0.9246>}
 FACIAL_R_NasolabialBulge {parent=FACIAL_C_FacialRoot, loc=<-3.4084 4.3621 5.7412>, rot=<0.0459 0.3776 -0.0189 -0.9246>}
 FACIAL_L_NasolabialFurrow {parent=FACIAL_C_FacialRoot, loc=<3.1337 3.3982 8.3427>, rot=<-0.0453 -0.1211 -0.3652 -0.9219>}
 FACIAL_R_NasolabialFurrow {parent=FACIAL_C_FacialRoot, loc=<-3.1436 3.399 8.344>, rot=<-0.0453 0.1211 0.3652 -0.9219>}
 FACIAL_L_CheekLower {parent=FACIAL_C_FacialRoot, loc=<3.7923 3.9969 5.0659>, rot=<0.0229 -0.4276 0.0522 -0.9022>}
 FACIAL_R_CheekLower {parent=FACIAL_C_FacialRoot, loc=<-3.7959 3.9997 5.0663>, rot=<0.0229 0.4276 -0.0522 -0.9022>}
 FACIAL_C_Nose {parent=FACIAL_C_FacialRoot, loc=<-0.0043 0.5373 8.9981>, rot=<-0.1432 0 0 -0.9897>}
 FACIAL_C_NoseLower {parent=FACIAL_C_Nose, loc=<0.0028 2.9444 -0.2556>, rot=<0.1305 0 0 -0.9914>}
 FACIAL_L_Nostril {parent=FACIAL_C_Nose, loc=<0.9056 2.2608 -0.2843>, rot=<0.0173 0.1305 0.0023 -0.9913>}
 FACIAL_R_Nostril {parent=FACIAL_C_Nose, loc=<-0.8958 2.2609 -0.2904>, rot=<0.0173 -0.1305 -0.0023 -0.9913>}
 FACIAL_C_NoseTip {parent=FACIAL_C_Nose, loc=<0.0015 1.9991 0.5825>, rot=<0.0872 0 0 -0.9962>}
 FACIAL_C_MouthUpper {parent=FACIAL_C_FacialRoot, loc=<-0.0066 4.6142 6.8389>, rot=<-0.013 0 0 -0.9999>}
 FACIAL_C_LipUpper {parent=FACIAL_C_MouthUpper, loc=<0.0097 0.095 2.07>, rot=<-0.0872 0 0 -0.9962>}
 FACIAL_L_LipUpper {parent=FACIAL_C_MouthUpper, loc=<1.1202 0.1812 1.9135>, rot=<-0.0778 -0.1601 -0.0379 -0.9833>}
 FACIAL_R_LipUpper {parent=FACIAL_C_MouthUpper, loc=<-1.1157 0.181 1.9122>, rot=<-0.0778 0.1601 0.0379 -0.9833>}
 FACIAL_L_LipUpperOuter {parent=FACIAL_C_MouthUpper, loc=<1.8737 0.2835 1.4685>, rot=<-0.0984 -0.2293 -0.0983 -0.9634>}
 FACIAL_R_LipUpperOuter {parent=FACIAL_C_MouthUpper, loc=<-1.8707 0.2841 1.4619>, rot=<-0.0984 0.2293 0.0983 -0.9634>}
 FACIAL_L_LipCorner {parent=FACIAL_C_MouthUpper, loc=<2.3804 0.6557 0.7796>, rot=<0 -0.2078 0 -0.9782>}
 FACIAL_R_LipCorner {parent=FACIAL_C_MouthUpper, loc=<-2.3781 0.6558 0.7684>, rot=<0 0.2078 0 -0.9782>}
 FACIAL_L_Masseter {parent=FACIAL_C_FacialRoot, loc=<6.0765 1.7662 3.8307>, rot=<0.0565 0.1128 0.0094 -0.992>}
 FACIAL_R_Masseter {parent=FACIAL_C_FacialRoot, loc=<-6.0845 1.7649 3.8295>, rot=<0.0565 -0.1128 -0.0093 -0.992>}
 FACIAL_C_UnderChin {parent=FACIAL_C_FacialRoot, loc=<0.0108 8.754 4.031>, rot=<0.4886 0 0 -0.8725>}
 FACIAL_L_UnderChin {parent=FACIAL_C_FacialRoot, loc=<2.7534 7.2857 2.8785>, rot=<0.3916 -0.3559 0.0777 -0.845>}
 FACIAL_R_UnderChin {parent=FACIAL_C_FacialRoot, loc=<-2.7573 7.2857 2.879>, rot=<0.3916 0.3559 -0.0777 -0.845>}
 FACIAL_C_LowerLipRotation {parent=FACIAL_C_FacialRoot, loc=<0.0018 2.4564 2.1209>, rot=<0.3298 0 0 -0.944>}
 FACIAL_C_MouthLower {parent=FACIAL_C_LowerLipRotation, loc=<-0.0082 -1.2495 5.0353>, rot=<-0.342 0 0 -0.9397>}
 FACIAL_C_LipLower {parent=FACIAL_C_MouthLower, loc=<0.0095 1.261 1.7291>, rot=<0.0872 0 0 -0.9962>}
 FACIAL_L_LipLowerOuter {parent=FACIAL_C_MouthLower, loc=<1.7697 1.0292 1.1616>, rot=<0.0913 -0.2321 0.0318 -0.9679>}
 FACIAL_R_LipLowerOuter {parent=FACIAL_C_MouthLower, loc=<-1.7654 1.0305 1.1585>, rot=<0.0913 0.2321 -0.0318 -0.9679>}
 FACIAL_C_Jaw {parent=FACIAL_C_FacialRoot, loc=<0.0018 2.4564 2.1209>, rot=<0.3298 0 0 -0.944>}
 FACIAL_C_Jawline {parent=FACIAL_C_Jaw, loc=<0.0091 2.48 7.1367>, rot=<0.2586 0 0 -0.966>}
 FACIAL_L_ChinSide {parent=FACIAL_C_Jaw, loc=<2.9238 0.5653 6.5195>, rot=<-0.2208 -0.3738 0.0499 -0.8995>}
 FACIAL_R_ChinSide {parent=FACIAL_C_Jaw, loc=<-2.9357 0.5668 6.519>, rot=<-0.2208 0.3738 -0.0499 -0.8995>}
 FACIAL_C_Chin {parent=FACIAL_C_Jaw, loc=<0.0014 0.9195 7.1151>, rot=<-0.2591 0 0 -0.9659>}
 FACIAL_L_Ear {parent=FACIAL_C_FacialRoot, loc=<6.4583 0.026 0.7913>, rot=<-0.0393 -0.0003 -0.0262 -0.9989>}
 FACIAL_R_Ear {parent=FACIAL_C_FacialRoot, loc=<-6.4666 0.0268 0.7875>, rot=<-0.0393 0.0004 0.0262 -0.9989>}
 FACIAL_L_Jawline {parent=FACIAL_C_FacialRoot, loc=<4.302 5.3309 4.3445>, rot=<0.1968 0.1585 -0.0526 -0.9661>}
 FACIAL_R_Jawline {parent=FACIAL_C_FacialRoot, loc=<-4.3089 5.3353 4.3425>, rot=<0.1968 -0.1584 0.0526 -0.9661>}
 FACIAL_C_NoseBridge {parent=FACIAL_C_FacialRoot, loc=<-0.0171 -1.5504 9.6001>, rot=<0.0308 0 0 -0.9995>}
 FACIAL_L_EyelidUpperB {parent=FACIAL_C_FacialRoot, loc=<2.921 -0.7747 6.9469>, rot=<-0.2376 0 0 -0.9714>}
 FACIAL_L_EyelidLowerB {parent=FACIAL_C_FacialRoot, loc=<2.921 -0.7747 6.9469>, rot=<0.3049 0 0 -0.9524>}
 FACIAL_R_EyelidLowerB {parent=FACIAL_C_FacialRoot, loc=<-2.9221 -0.7719 6.9485>, rot=<0.3132 0 0 -0.9497>}
 FACIAL_R_EyelidUpperB {parent=FACIAL_C_FacialRoot, loc=<-2.9221 -0.7719 6.9485>, rot=<-0.2376 0 0 -0.9714>}
 clavicle_l {parent=spine_05, loc=<5.4831 -0.1053 -3.2946>, rot=<0.0465 -0.7289 -0.009 -0.6829>}
 upperarm_l {parent=clavicle_l, loc=<9.8658 -1.2136 -0.4105>, rot=<0.0164 -0.3245 -0.0269 -0.9454>}
 clavicle_r {parent=spine_05, loc=<5.4831 -0.1053 3.2946>, rot=<0.7289 0.0465 -0.6829 0.009>}
 upperarm_r {parent=clavicle_r, loc=<-9.8658 1.2137 0.4103>, rot=<0.0164 -0.3245 -0.0269 -0.9454>}
 dyn_cape_collider {parent=spine_05, loc=<0 -0.3005 0>, rot=<0.0284 -0.7065 -0.0284 -0.7065>}
]
variables [
 Crouch {type=bool, default=False}
 Moving {type=bool, default=False}
 clavicle_r_SPR_back_alpha {type=double, default=0}
 head_SPR_look_down_alpha {type=double, default=0}
 head_SPR_look_left_alpha {type=double, default=0}
 head_SPR_look_right_alpha {type=double, default=0}
 head_SPR_look_up_alpha {type=double, default=0}
 look_down_upperarm_r_back_multiplied_alpha {type=double, default=0}
 look_up_and_left_two_hands_multiplied_alpha {type=double, default=0}
 look_up_two_hands_rifle_multiplied_alpha {type=double, default=0}
 uppearm_r_SPR_aim_up_two_hands_alpha {type=double, default=0}
 upperarm_r_SPR_back_alpha {type=double, default=0}
 upperarm_r_down_alpha {type=double, default=0}
 upperarm_r_forward_alpha {type=double, default=0}
 upperarm_r_rifle_up_alpha {type=double, default=0}
 upperarm_r_skydiving {type=double, default=0}
]
