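# Physics asset for the front hair on the face accessory part of the female medium MotorMonth player character body.
# Capsule bodies cover the head and two hair chains: four left bones, and a right master bone leading through one segment to two hair-tip bones.
# Joints to the head, plus the right master-to-first-segment joint, leave swing and twist free.
# The remaining joints set moderate swing limits with very tight twist, letting strands bend with little roll.
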PhysicsAsset /BRCosmetics/Characters/Player/Female/Medium/Bodies/F_MED_MotorMonth/Meshes/Parts/F_MED_MotorMonth_Front_FaceAcc_Physics
bodies {
 dyn_hair_front_l_01 [{
   shapeType=Capsule
   length=1.5706
   radius=1.0553
   shapeName=dyn_hair_front_l_01_capsule
   center=<1.909 0.3498 -0.195>
   rotation=<rot 0 89.6025 -67.6027>
  }]
 dyn_hair_front_l_02 [{shapeType=Capsule, length=3.9338, radius=1.5662, shapeName=dyn_hair_front_l_02_capsule, center=<1.179 0.5 0>, rotation=<rot 80 9.9999 0>}]
 dyn_hair_front_l_03 [{shapeType=Capsule, length=3, radius=0.7, shapeName=dyn_hair_front_l_03_capsule, center=<1 0 0>, rotation=<rot 90 0 0>}]
 dyn_hair_front_l_04 [{shapeType=Capsule, length=2.5, radius=0.7, shapeName=dyn_hair_front_l_04_capsule, center=<1 0 0>, rotation=<rot 90 0 0>}]
 dyn_hair_front_r_01 [{
   shapeType=Capsule
   length=2.8157
   radius=1.6739
   shapeName=dyn_hair_front_r_01_capsule
   center=<4.4319 0.0962 1.2979>
   rotation=<rot 0 113.9391 -88.9376>
  }]
 dyn_hair_front_r_MSTR [{shapeType=Capsule, length=0.505, radius=0.505, shapeName=dyn_hair_front_r_MSTR_capsule, rotation=<rot 0 -90 0>}]
 dyn_hair_front_r_hairTips_01 [{
   shapeType=Capsule
   length=4.0754
   radius=0.5491
   shapeName=dyn_hair_front_r_hairTips_01_capsule
   center=<2.5464 0.396 -0.2494>
   rotation=<rot 0 -84.9128 -80.9475>
  }]
 dyn_hair_front_r_hairTips_02 [{
   shapeType=Capsule
   length=3.9365
   radius=0.5
   shapeName=dyn_hair_front_r_hairTips_02_capsule
   center=<2.5962 -0.9551 1.2623>
   rotation=<rot 0 58.9611 -54.1975>
  }]
 head [{shapeType=Capsule, length=6.1965, radius=6.5617, shapeName=head_capsule, center=<2.5967 2.5691 0>, rotation=<rot 90 0 0>}]
}
constraints [
 {
  bone1Name=dyn_hair_front_r_MSTR
  bone2Name=head
  swing1LimitDegrees=45
  swing1Motion=Free
  swing2LimitDegrees=45
  swing2Motion=Free
  twistLimitDegrees=45
  twistMotion=Free
 }
 {bone1Name=dyn_hair_front_l_02, bone2Name=dyn_hair_front_l_01, swing1LimitDegrees=30, swing2LimitDegrees=30, twistLimitDegrees=5}
 {bone1Name=dyn_hair_front_l_03, bone2Name=dyn_hair_front_l_02, swing1LimitDegrees=45, swing2LimitDegrees=45, twistLimitDegrees=5}
 {bone1Name=dyn_hair_front_l_04, bone2Name=dyn_hair_front_l_03, swing1LimitDegrees=45, swing2LimitDegrees=45, twistLimitDegrees=5}
 {
  bone1Name=dyn_hair_front_l_01
  bone2Name=head
  swing1LimitDegrees=45
  swing1Motion=Free
  swing2LimitDegrees=45
  swing2Motion=Free
  twistLimitDegrees=45
  twistMotion=Free
 }
 {
  bone1Name=dyn_hair_front_r_hairTips_02
  bone2Name=dyn_hair_front_r_hairTips_01
  swing1LimitDegrees=45
  swing2LimitDegrees=45
  twistLimitDegrees=5
 }
 {bone1Name=dyn_hair_front_r_hairTips_01, bone2Name=dyn_hair_front_r_01, swing1LimitDegrees=45, swing2LimitDegrees=45, twistLimitDegrees=5}
 {
  bone1Name=dyn_hair_front_r_01
  bone2Name=dyn_hair_front_r_MSTR
  swing1LimitDegrees=45
  swing1Motion=Free
  swing2LimitDegrees=45
  swing2Motion=Free
  twistLimitDegrees=45
  twistMotion=Free
 }
]
