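# Post-physics-corrected Anim Blueprint for the female MotorMonth FaceAcc part, animating its hair over poses copied from the body and attached parent.
# RigidBody simulations, swapped for front end, feed AnimDynamics tips, corrective constraints and a crouch/move-aware Control Rig.
# Cape and backbling flags offset bones, and constraints pin hair to head, neck and spine during backpack sprinting or skydiving.
# The EventGraph tunes gravity and inertia scales by state.
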AnimBlueprint /BRCosmetics/Characters/Player/Female/Medium/Bodies/F_MED_MotorMonth/Meshes/Parts/F_MED_MotorMonth_FaceAcc_AnimBP
target_skeleton=&/F_MED_MotorMonth_FaceAcc_Skeleton
graphs [
 AnimGraph {nodes [
   n1 = Copy Pose From Mesh(SourceMeshComponent=Get Body_Mesh())
   n2 = Copy Pose From Mesh() {bUseAttachedParent=true}
   n3 = Get isWearCape() -> (isWearCape:bool)
   n4 = "Transform (Modify) Bone - Bone: dyn_hair_bk_MSTR"!(
    ComponentPose=Local To Component(LocalPose=Layered blend per bone(BasePose=@n2, BlendPoses_0=@n1) {layerSetup [{branchFilters [{boneName=upperarm_l}, {boneName=upperarm_r}]}]})
    bAlphaBoolEnabled:bool=@n3
    Translation="0,0,0"
    Rotation="8,0.000,0.000"
    Scale="1,1,1"
   ) -> (Pose) {
    alphaInputType=Bool
    rotationMode=BMM_Additive
    rotationSpace=BCS_BoneSpace
    translationMode=BMM_Additive
    translationSpace=BCS_BoneSpace
    alphaBoolBlend {blendOption=Linear}
    boneToModify {boneName=dyn_hair_bk_MSTR}
    rotation {pitch=8}
   }
   n5 = "Transform (Modify) Bone - Bone: dyn_hair_r_MSTR"!(ComponentPose=@n4, bAlphaBoolEnabled:bool=@n3, Translation="0,0,0", Rotation="5,0,0", Scale="1,1,1") -> (Pose) {
    alphaInputType=Bool
    rotationMode=BMM_Additive
    rotationSpace=BCS_BoneSpace
    translationMode=BMM_Additive
    translationSpace=BCS_BoneSpace
    alphaBoolBlend {blendOption=Linear}
    boneToModify {boneName=dyn_hair_r_MSTR}
    rotation {pitch=5}
   }
   n6 = "Transform (Modify) Bone - Bone: dyn_hair_l_MSTR"!(ComponentPose=@n5, bAlphaBoolEnabled:bool=@n3, Translation="0,0,0", Rotation="5,0,0", Scale="1,1,1") -> (Pose) {
    alphaInputType=Bool
    rotationMode=BMM_Additive
    rotationSpace=BCS_BoneSpace
    translationMode=BMM_Additive
    translationSpace=BCS_BoneSpace
    alphaBoolBlend {blendOption=Linear}
    boneToModify {boneName=dyn_hair_l_MSTR}
    rotation {pitch=5}
   }
   n7 = "Transform (Modify) Bone - Bone: dyn_cape_collider"!(ComponentPose=@n6, bAlphaBoolEnabled:bool=@n3, Translation="-5.910,-14.689,2.149", Rotation="-0.004,-20.000,-0.000", Scale="1,1,1") -> (Pose) {
    alphaInputType=Bool
    rotationMode=BMM_Additive
    rotationSpace=BCS_BoneSpace
    translationMode=BMM_Additive
    translationSpace=BCS_BoneSpace
    alphaBoolBlend {blendOption=Linear}
    boneToModify {boneName=dyn_cape_collider}
    rotation=<rot -0.0039 -20 -0.0002>
    translation=<-5.91 -14.689 2.149>
   }
   n8 = Get AccScaleOverride() -> (AccScaleOverride:Vector)
   n9 = Get GravityOverride()
   n10 = Get VelScaleOverride() -> (VelScaleOverride:Vector)
   n11 = RigidBody!(
    ComponentPose=@n7
    bAlphaBoolEnabled:bool=NOT Boolean(A=Get bIsInFrontEnd())
    OverrideWorldGravity:Vector=@n9
    ExternalForce
    ComponentLinearAccScale:Vector=@n8
    ComponentLinearVelScale:Vector=@n10
   ) -> (Pose) {
    alphaInputType=Bool
    bClampLinearTranslationLimitToRefPose=true
    bOverrideWorldGravity=true
    overridePhysicsAsset=&/F_MED_MotorMonth_FaceAcc_Physics
    simulationSpace=BaseBoneSpace
    alphaBoolBlend {blendOption=Linear}
    alphaScaleBiasClamp {interpSpeedDecreasing=4, interpSpeedIncreasing=6}
    baseBoneRef {boneName=spine_05}
    componentAppliedLinearAccClamp=<3500 3500 3500>
    componentLinearAccScale=<0.6 0.6 0.2>
    componentLinearVelScale=<0.5 0.5 0.5>
    overrideWorldGravity {z=-875}
   }
   n12 = RigidBody!(ComponentPose=@n11, Alpha:float, OverrideWorldGravity=@n9, ExternalForce) -> (Pose) {
    bClampLinearTranslationLimitToRefPose=true
    bOverrideWorldGravity=true
    overridePhysicsAsset=&/F_MED_MotorMonth_Front_FaceAcc_Physics
    simulationSpace=BaseBoneSpace
    alphaBoolBlend {blendOption=Linear}
    baseBoneRef {boneName=spine_05}
    componentAppliedLinearAccClamp=<3500 3500 3500>
    componentLinearAccScale=<0.25 0.25 0.25>
    componentLinearVelScale=<0.25 0.25 0.25>
    overrideWorldGravity {z=-980}
   }
   n13 = RigidBody!(
    ComponentPose=@n12
    bAlphaBoolEnabled:bool=Get bIsInFrontEnd()
    OverrideWorldGravity=@n9
    ExternalForce
    ComponentLinearAccScale:Vector=@n8
    ComponentLinearVelScale:Vector=@n10
   ) -> (Pose) {
    alphaInputType=Bool
    bClampLinearTranslationLimitToRefPose=true
    bOverrideWorldGravity=true
    overridePhysicsAsset=&/F_MED_MotorMonth_FaceAcc_FrontEnd_Physics
    simulationSpace=BaseBoneSpace
    alphaBoolBlend {blendOption=Linear}
    alphaScaleBiasClamp {interpSpeedDecreasing=4, interpSpeedIncreasing=6}
    baseBoneRef {boneName=spine_05}
    componentAppliedLinearAccClamp=<3500 3500 3500>
    componentLinearAccScale=<0.6 0.6 0.2>
    componentLinearVelScale=<0.5 0.5 0.5>
    overrideWorldGravity {z=-875}
   }
   n14 = Get bIsInFrontEnd()
   n15 = "Transform (Modify) Bone - Bone: dyn_hairTips_r_03"!(ComponentPose=@n13, bAlphaBoolEnabled:bool=@n14, Translation="1,0,1", Rotation="0,0,0", Scale="0,1,1") -> (Pose) {
    alphaInputType=Bool
    rotationMode=BMM_Additive
    rotationSpace=BCS_BoneSpace
    translationMode=BMM_Additive
    translationSpace=BCS_BoneSpace
    alphaBoolBlend {blendOption=Linear}
    boneToModify {boneName=dyn_hairTips_r_03}
    scale {x=0}
    translation {x=1, z=1}
   }
   n16 = "Transform (Modify) Bone - Bone: dyn_hair_bk_03"!(ComponentPose=@n15, bAlphaBoolEnabled:bool=@n14, Translation="0,-1,0", Rotation="-5,0,0", Scale="1,1,1") -> (Pose) {
    alphaInputType=Bool
    rotationMode=BMM_Additive
    rotationSpace=BCS_BoneSpace
    translationMode=BMM_Additive
    translationSpace=BCS_BoneSpace
    alphaBoolBlend {blendOption=Linear}
    boneToModify {boneName=dyn_hair_bk_03}
    rotation {pitch=-5}
    translation {y=-1}
   }
   n17 = "Transform (Modify) Bone - Bone: dyn_hair_bk_02"!(ComponentPose=@n16, bAlphaBoolEnabled:bool=@n14, Translation="0,-1,0", Rotation="0,-5,0", Scale="1,1,1") -> (Pose) {
    alphaInputType=Bool
    rotationMode=BMM_Additive
    rotationSpace=BCS_BoneSpace
    translationMode=BMM_Additive
    translationSpace=BCS_BoneSpace
    alphaBoolBlend {blendOption=Linear}
    boneToModify {boneName=dyn_hair_bk_02}
    rotation {yaw=-5}
    translation {y=-1}
   }
   n18 = "Transform (Modify) Bone - Bone: dyn_hair_l_01"!(ComponentPose=@n17, bAlphaBoolEnabled:bool=@n14, Translation="0,-1,0", Rotation="0,-25,0", Scale="0,1,1") -> (Pose) {
    alphaInputType=Bool
    rotationMode=BMM_Additive
    rotationSpace=BCS_BoneSpace
    translationMode=BMM_Additive
    translationSpace=BCS_BoneSpace
    alphaBoolBlend {blendOption=Linear}
    boneToModify {boneName=dyn_hair_l_01}
    rotation {yaw=-25}
    scale {x=0}
    translation {y=-1}
   }
   n19 = "Transform (Modify) Bone - Bone: dyn_hair_l_04"!(ComponentPose=@n18, bAlphaBoolEnabled:bool=@n14, Translation="0,0,0", Rotation="50,0,0", Scale="0,1,1") -> (Pose) {
    alphaInputType=Bool
    rotationMode=BMM_Additive
    rotationSpace=BCS_BoneSpace
    translationMode=BMM_Additive
    translationSpace=BCS_BoneSpace
    alphaBoolBlend {blendOption=Linear}
    boneToModify {boneName=dyn_hair_l_04}
    rotation {pitch=50}
    scale {x=0}
   }
   n20 = Anim Dynamics - Bone: dyn_hairTips_l_01(ComponentPose=@n19, Alpha=1, ExternalForce="0,0,0") {
    angularDampingOverride=1
    angularSpringConstant=6000
    bOverrideAngularDamping=true
    bUseGravityOverride=true
    simulationSpace=RootRelative
    alphaBoolBlend {blendOption=Linear}
    boundBone {boneName=dyn_hairTips_l_01}
    componentLinearAccScale=<1 1 1>
    componentLinearVelScale=<1 1 1>
    gravityOverride {z=-980}
    physicsBodyDefinitions [{
      boundBone {boneName=dyn_hairTips_l_01}
      boxExtents=<10 10 10>
      constraintSetup {angularLimitsMax=<10 10 10>, angularLimitsMin=<-10 -10 0>, angularTarget=<1 0 0>}
      localJointOffset=<1 0 0>
     }]
    relativeSpaceBone {boneName=spine_05}
   }
   n21 = Anim Dynamics - Bone: dyn_hairTips_l_02(ComponentPose=@n20, Alpha=1, ExternalForce="0,0,0") {
    angularDampingOverride=1
    angularSpringConstant=6000
    bOverrideAngularDamping=true
    bUseGravityOverride=true
    simulationSpace=RootRelative
    alphaBoolBlend {blendOption=Linear}
    boundBone {boneName=dyn_hairTips_l_02}
    componentLinearAccScale=<1 1 1>
    componentLinearVelScale=<1 1 1>
    gravityOverride {z=-980}
    physicsBodyDefinitions [{
      boundBone {boneName=dyn_hairTips_l_02}
      boxExtents=<10 10 10>
      constraintSetup {angularLimitsMax=<10 10 10>, angularLimitsMin=<-10 -10 0>, angularTarget=<1 0 0>}
      localJointOffset=<1 0 0>
     }]
    relativeSpaceBone {boneName=spine_05}
   }
   n22 = Anim Dynamics - Bone: dyn_hairTips_l_03(ComponentPose=@n21, Alpha=1, ExternalForce="0,0,0") {
    angularDampingOverride=1
    angularSpringConstant=6000
    bOverrideAngularDamping=true
    bUseGravityOverride=true
    simulationSpace=RootRelative
    alphaBoolBlend {blendOption=Linear}
    boundBone {boneName=dyn_hairTips_l_03}
    componentLinearAccScale=<1 1 1>
    componentLinearVelScale=<1 1 1>
    gravityOverride {z=-980}
    physicsBodyDefinitions [{
      boundBone {boneName=dyn_hairTips_l_03}
      boxExtents=<10 10 10>
      constraintSetup {angularLimitsMax=<15 0 15>, angularLimitsMin=<-15 -25 0>, angularTarget=<1 0 0>}
      localJointOffset=<1 0 0>
     }]
    relativeSpaceBone {boneName=spine_05}
   }
   n23 = Anim Dynamics - Bone: dyn_hairTips_l_04(ComponentPose=@n22, Alpha=1, ExternalForce="0,0,0") {
    angularDampingOverride=1
    angularSpringConstant=6000
    bOverrideAngularDamping=true
    bUseGravityOverride=true
    simulationSpace=RootRelative
    alphaBoolBlend {blendOption=Linear}
    boundBone {boneName=dyn_hairTips_l_04}
    componentLinearAccScale=<1 1 1>
    componentLinearVelScale=<1 1 1>
    gravityOverride {z=-980}
    physicsBodyDefinitions [{
      boundBone {boneName=dyn_hairTips_l_04}
      boxExtents=<10 10 10>
      constraintSetup {angularLimitsMax=<15 0 15>, angularLimitsMin=<-15 -25 0>, angularTarget=<1 0 0>}
      localJointOffset=<1 0 0>
     }]
    relativeSpaceBone {boneName=spine_05}
   }
   n24 = Anim Dynamics - Bone: dyn_hairTips_r_01(ComponentPose=@n23, Alpha=1, ExternalForce="0,0,0") {
    angularDampingOverride=1
    angularSpringConstant=6000
    bOverrideAngularDamping=true
    bUseGravityOverride=true
    simulationSpace=RootRelative
    alphaBoolBlend {blendOption=Linear}
    boundBone {boneName=dyn_hairTips_r_01}
    componentLinearAccScale=<1 1 1>
    componentLinearVelScale=<1 1 1>
    gravityOverride {z=-980}
    physicsBodyDefinitions [{
      boundBone {boneName=dyn_hairTips_r_01}
      boxExtents=<10 10 10>
      constraintSetup {angularLimitsMax=<10 10 10>, angularLimitsMin=<-10 -10 -10>, angularTarget=<1 0 0>}
      localJointOffset=<1 0 0>
     }]
    relativeSpaceBone {boneName=spine_05}
   }
   n25 = Anim Dynamics - Bone: dyn_hairTips_r_04(ComponentPose=@n24, Alpha=1, ExternalForce="0,0,0") {
    angularDampingOverride=1
    angularSpringConstant=6000
    bOverrideAngularDamping=true
    bUseGravityOverride=true
    simulationSpace=RootRelative
    alphaBoolBlend {blendOption=Linear}
    boundBone {boneName=dyn_hairTips_r_04}
    componentLinearAccScale=<1 1 1>
    componentLinearVelScale=<1 1 1>
    gravityOverride {z=-980}
    physicsBodyDefinitions [{
      boundBone {boneName=dyn_hairTips_r_04}
      boxExtents=<10 10 10>
      constraintSetup {angularLimitsMax=<10 80 10>, angularLimitsMin=<-10 -40 -10>, angularTarget=<1 0 0>}
      localJointOffset=<1 0 0>
     }]
    relativeSpaceBone {boneName=spine_05}
   }
   n26 = Anim Dynamics - Bone: dyn_hairTips_r_02(ComponentPose=@n25, Alpha=1, ExternalForce="0,0,0") {
    angularDampingOverride=1
    angularSpringConstant=6000
    bOverrideAngularDamping=true
    bUseGravityOverride=true
    simulationSpace=RootRelative
    alphaBoolBlend {blendOption=Linear}
    boundBone {boneName=dyn_hairTips_r_02}
    componentLinearAccScale=<1 1 1>
    componentLinearVelScale=<1 1 1>
    gravityOverride {z=-980}
    physicsBodyDefinitions [{
      boundBone {boneName=dyn_hairTips_r_02}
      boxExtents=<10 10 10>
      constraintSetup {angularLimitsMax=<10 2 10>, angularLimitsMin=<-10 -20 -10>, angularTarget=<1 0 0>}
      localJointOffset=<1 0 0>
     }]
    relativeSpaceBone {boneName=spine_05}
   }
   n27 = Anim Dynamics - Bone: dyn_hairTips_r_03(ComponentPose=@n26, Alpha=1, ExternalForce="0,0,0") {
    angularDampingOverride=1
    angularSpringConstant=6000
    bOverrideAngularDamping=true
    bUseGravityOverride=true
    simulationSpace=RootRelative
    alphaBoolBlend {blendOption=Linear}
    boundBone {boneName=dyn_hairTips_r_03}
    componentLinearAccScale=<1 1 1>
    componentLinearVelScale=<1 1 1>
    gravityOverride {z=-980}
    physicsBodyDefinitions [{
      boundBone {boneName=dyn_hairTips_r_03}
      boxExtents=<10 10 10>
      constraintSetup {angularLimitsMax=<10 2 10>, angularLimitsMin=<-10 -20 -10>, angularTarget=<1 0 0>}
      localJointOffset=<1 0 0>
     }]
    relativeSpaceBone {boneName=spine_05}
   }
   n28 = "Constraint - dyn_hair_front_r_02 (P)"(ComponentPose=@n27, Alpha=1, ConstraintWeights_0=1) {
    alphaBoolBlend {blendOption=Linear}
    boneToModify {boneName=dyn_hair_front_r_02}
    constraintSetup [{targetBone {boneName=dyn_hair_r_01}}]
    constraintWeights [1]
   }
   n29 = "Constraint - dyn_hair_front_r_03 (P)"(ComponentPose=@n28, Alpha=1, ConstraintWeights_0=1) {
    alphaBoolBlend {blendOption=Linear}
    boneToModify {boneName=dyn_hair_front_r_03}
    constraintSetup [{targetBone {boneName=dyn_hair_r_02}}]
    constraintWeights [1]
   }
   n30 = Get BackPackTactJump() -> (BackPackTactJump:double)
   n31 = "Constraint - dyn_hair_r_01 (P)"(ComponentPose=@n29, Alpha=@n30, ConstraintWeights_0=1) {
    alphaBoolBlend {blendOption=Linear}
    boneToModify {boneName=dyn_hair_r_01}
    constraintSetup [{targetBone {boneName=head}}]
    constraintWeights [1]
   }
   n32 = "Constraint - dyn_hair_bk_01 (P)"(ComponentPose=@n31, Alpha=@n30, ConstraintWeights_0=1) {
    alphaBoolBlend {blendOption=Linear}
    boneToModify {boneName=dyn_hair_bk_01}
    constraintSetup [{targetBone {boneName=head}}]
    constraintWeights [1]
   }
   n33 = "Constraint - dyn_hair_l_01 (P)"(ComponentPose=@n32, Alpha=@n30, ConstraintWeights_0=1) {
    alphaBoolBlend {blendOption=Linear}
    boneToModify {boneName=dyn_hair_l_01}
    constraintSetup [{targetBone {boneName=head}}]
    constraintWeights [1]
   }
   n34 = "Constraint - dyn_hair_r_02 (P)"(ComponentPose=@n33, Alpha=@n30, ConstraintWeights_0=1) {
    alphaBoolBlend {blendOption=Linear}
    boneToModify {boneName=dyn_hair_r_02}
    constraintSetup [{targetBone {boneName=neck_01}}]
    constraintWeights [1]
   }
   n35 = "Constraint - dyn_hair_bk_02 (P)"(ComponentPose=@n34, Alpha=@n30, ConstraintWeights_0=1) {
    alphaBoolBlend {blendOption=Linear}
    boneToModify {boneName=dyn_hair_bk_02}
    constraintSetup [{targetBone {boneName=neck_01}}]
    constraintWeights [1]
   }
   n36 = "Constraint - dyn_hair_l_02 (P)"(ComponentPose=@n35, Alpha=@n30, ConstraintWeights_0=1) {
    alphaBoolBlend {blendOption=Linear}
    boneToModify {boneName=dyn_hair_l_02}
    constraintSetup [{targetBone {boneName=neck_01}}]
    constraintWeights [1]
   }
   n37 = "Constraint - dyn_hair_r_03 (P)"(ComponentPose=@n36, Alpha=@n30, ConstraintWeights_0=1) {
    alphaBoolBlend {blendOption=Linear}
    boneToModify {boneName=dyn_hair_r_03}
    constraintSetup [{targetBone {boneName=spine_05}}]
    constraintWeights [1]
   }
   n38 = "Constraint - dyn_hair_r_04 (P)"(ComponentPose=@n37, Alpha=@n30, ConstraintWeights_0=1) {
    alphaBoolBlend {blendOption=Linear}
    boneToModify {boneName=dyn_hair_r_04}
    constraintSetup [{targetBone {boneName=spine_04}}]
    constraintWeights [1]
   }
   n39 = "Constraint - dyn_hair_bk_03 (P)"(ComponentPose=@n38, Alpha=@n30, ConstraintWeights_0=1) {
    alphaBoolBlend {blendOption=Linear}
    boneToModify {boneName=dyn_hair_bk_03}
    constraintSetup [{targetBone {boneName=spine_05}}]
    constraintWeights [1]
   }
   n40 = "Constraint - dyn_hair_bk_04 (P)"(ComponentPose=@n39, Alpha=@n30, ConstraintWeights_0=1) {
    alphaBoolBlend {blendOption=Linear}
    boneToModify {boneName=dyn_hair_bk_04}
    constraintSetup [{targetBone {boneName=spine_04}}]
    constraintWeights [1]
   }
   n41 = "Constraint - dyn_hair_l_03 (P)"(ComponentPose=@n40, Alpha=@n30, ConstraintWeights_0=1) {
    alphaBoolBlend {blendOption=Linear}
    boneToModify {boneName=dyn_hair_l_03}
    constraintSetup [{targetBone {boneName=spine_05}}]
    constraintWeights [1]
   }
   n42 = "Constraint - dyn_hair_l_04 (P)"(ComponentPose=@n41, Alpha=@n30, ConstraintWeights_0=1) {
    alphaBoolBlend {blendOption=Linear}
    boneToModify {boneName=dyn_hair_l_04}
    constraintSetup [{targetBone {boneName=spine_04}}]
    constraintWeights [1]
   }
   n43 = Get SkydivePack() -> (SkydivePack:double)
   n44 = "Constraint - dyn_hair_r_01 (P)"(ComponentPose=@n42, Alpha=@n43, ConstraintWeights_0=1) {
    alphaBoolBlend {blendOption=Linear}
    boneToModify {boneName=dyn_hair_r_01}
    constraintSetup [{targetBone {boneName=head}}]
    constraintWeights [1]
   }
   n45 = "Constraint - dyn_hair_bk_01 (P)"(ComponentPose=@n44, Alpha=@n43, ConstraintWeights_0=1) {
    alphaBoolBlend {blendOption=Linear}
    boneToModify {boneName=dyn_hair_bk_01}
    constraintSetup [{targetBone {boneName=head}}]
    constraintWeights [1]
   }
   n46 = "Constraint - dyn_hair_l_01 (P)"(ComponentPose=@n45, Alpha=@n43, ConstraintWeights_0=1) {
    alphaBoolBlend {blendOption=Linear}
    boneToModify {boneName=dyn_hair_l_01}
    constraintSetup [{targetBone {boneName=head}}]
    constraintWeights [1]
   }
   n47 = "Constraint - dyn_hair_r_02 (P)"(ComponentPose=@n46, Alpha=@n43, ConstraintWeights_0=1) {
    alphaBoolBlend {blendOption=Linear}
    boneToModify {boneName=dyn_hair_r_02}
    constraintSetup [{targetBone {boneName=neck_01}}]
    constraintWeights [1]
   }
   n48 = "Constraint - dyn_hair_bk_02 (P)"(ComponentPose=@n47, Alpha=@n43, ConstraintWeights_0=1) {
    alphaBoolBlend {blendOption=Linear}
    boneToModify {boneName=dyn_hair_bk_02}
    constraintSetup [{targetBone {boneName=neck_01}}]
    constraintWeights [1]
   }
   n49 = "Constraint - dyn_hair_l_02 (P)"(ComponentPose=@n48, Alpha=@n43, ConstraintWeights_0=1) {
    alphaBoolBlend {blendOption=Linear}
    boneToModify {boneName=dyn_hair_l_02}
    constraintSetup [{targetBone {boneName=neck_01}}]
    constraintWeights [1]
   }
   n50 = "Constraint - dyn_hair_r_03 (P)"(ComponentPose=@n49, Alpha=@n43, ConstraintWeights_0=1) {
    alphaBoolBlend {blendOption=Linear}
    boneToModify {boneName=dyn_hair_r_03}
    constraintSetup [{targetBone {boneName=spine_05}}]
    constraintWeights [1]
   }
   n51 = "Constraint - dyn_hair_r_04 (P)"(ComponentPose=@n50, Alpha=@n43, ConstraintWeights_0=1) {
    alphaBoolBlend {blendOption=Linear}
    boneToModify {boneName=dyn_hair_r_04}
    constraintSetup [{targetBone {boneName=spine_04}}]
    constraintWeights [1]
   }
   n52 = "Constraint - dyn_hair_bk_03 (P)"(ComponentPose=@n51, Alpha=@n43, ConstraintWeights_0=1) {
    alphaBoolBlend {blendOption=Linear}
    boneToModify {boneName=dyn_hair_bk_03}
    constraintSetup [{targetBone {boneName=spine_05}}]
    constraintWeights [1]
   }
   n53 = "Constraint - dyn_hair_bk_04 (P)"(ComponentPose=@n52, Alpha=@n43, ConstraintWeights_0=1) {
    alphaBoolBlend {blendOption=Linear}
    boneToModify {boneName=dyn_hair_bk_04}
    constraintSetup [{targetBone {boneName=spine_04}}]
    constraintWeights [1]
   }
   n54 = "Constraint - dyn_hair_l_03 (P)"(ComponentPose=@n53, Alpha=@n43, ConstraintWeights_0=1) {
    alphaBoolBlend {blendOption=Linear}
    boneToModify {boneName=dyn_hair_l_03}
    constraintSetup [{targetBone {boneName=spine_05}}]
    constraintWeights [1]
   }
   n55 = "Constraint - dyn_hair_l_04 (P)"(ComponentPose=@n54, Alpha=@n43, ConstraintWeights_0=1) {
    alphaBoolBlend {blendOption=Linear}
    boneToModify {boneName=dyn_hair_l_04}
    constraintSetup [{targetBone {boneName=spine_04}}]
    constraintWeights [1]
   }
   n56 = "Transform (Modify) Bone - Bone: dyn_hair_r_02"(ComponentPose=@n55, Alpha=@n43, Translation="0,0,5", Rotation="5,5,15", Scale="1,1,1") {
    rotationMode=BMM_Additive
    rotationSpace=BCS_BoneSpace
    translationMode=BMM_Additive
    translationSpace=BCS_BoneSpace
    alphaBoolBlend {blendOption=Linear}
    boneToModify {boneName=dyn_hair_r_02}
    rotation=<rot 5 5 15>
    translation {z=5}
   }
   n57 = "Transform (Modify) Bone - Bone: dyn_hair_l_02"(ComponentPose=@n56, Alpha=@n43, Translation="0,0,5", Rotation="0,-5,-15", Scale="1,1,1") {
    rotationMode=BMM_Additive
    rotationSpace=BCS_BoneSpace
    translationMode=BMM_Additive
    translationSpace=BCS_BoneSpace
    alphaBoolBlend {blendOption=Linear}
    boneToModify {boneName=dyn_hair_l_02}
    rotation {roll=-15, yaw=-5}
    translation {z=5}
   }
   n58 = "Transform (Modify) Bone - Bone: dyn_hair_bk_02"(ComponentPose=@n57, bAlphaBoolEnabled=Get isWearBackbling() -> (isWearBackbling:bool), Translation="0,0,0", Rotation="-5,0,0", Scale="1,1,1") {
    alphaInputType=Bool
    rotationMode=BMM_Additive
    rotationSpace=BCS_BoneSpace
    translationMode=BMM_Additive
    translationSpace=BCS_BoneSpace
    alphaBoolBlend {blendOption=Linear}
    boneToModify {boneName=dyn_hair_bk_02}
    rotation {pitch=-5}
   }
   n59 = "Constraint - dyn_hair_r_03 (PP)"!(ComponentPose=@n58, Alpha=0.5, ConstraintWeights_0=0.5, ConstraintWeights_1:float=0.5) -> (Pose) {
    alphaBoolBlend {blendOption=Linear}
    boneToModify {boneName=dyn_hair_r_03}
    constraintSetup [{targetBone {boneName=dyn_hair_bk_03}}, {targetBone {boneName=clavicle_r}}]
    constraintWeights [1,1]
   }
   n60 = "Constraint - dyn_hair_l_03 (P)"(ComponentPose=@n59, Alpha=0.5, ConstraintWeights_0=1) {
    alphaBoolBlend {blendOption=Linear}
    boneToModify {boneName=dyn_hair_l_03}
    constraintSetup [{targetBone {boneName=dyn_hair_bk_03}}]
    constraintWeights [1]
   }
   n61 = "Constraint - dyn_hair_l_04 (P)"(ComponentPose=@n60, Alpha=0.8, ConstraintWeights_0=1) {
    alphaBoolBlend {blendOption=Linear}
    boneToModify {boneName=dyn_hair_l_04}
    constraintSetup [{targetBone {boneName=dyn_hair_bk_04}}]
    constraintWeights [1]
   }
   n62 = "Constraint - dyn_hair_r_04 (P)"(ComponentPose=@n61, Alpha=0.3, ConstraintWeights_0=1) {
    alphaBoolBlend {blendOption=Linear}
    boneToModify {boneName=dyn_hair_r_04}
    constraintSetup [{targetBone {boneName=dyn_hair_bk_04}}]
    constraintWeights [1]
   }
   n63 = "Transform (Modify) Bone - Bone: dyn_hair_front_r_04"(ComponentPose=@n62, Alpha=1, Translation="0,0,0", Rotation="-5,0,0", Scale="1,1,1") {
    rotationMode=BMM_Additive
    rotationSpace=BCS_BoneSpace
    translationMode=BMM_Additive
    translationSpace=BCS_BoneSpace
    alphaBoolBlend {blendOption=Linear}
    boneToModify {boneName=dyn_hair_front_r_04}
    rotation {pitch=-5}
   }
   n64 = Control Rig!(Source=Component To Local(ComponentPose=@n63), Alpha, Crouch:bool=Get bIsCrouching(), Moving:bool=Get bIsMoving2D()) -> (Pose) {
    alphaBoolBlend {blendOption=Linear}
    controlRigAssetReference {blueprintRigClass=&/F_MED_MotorMonth_FaceAcc_CtrlRig.F_MED_MotorMonth_FaceAcc_CtrlRig_C}
   }
   Output Pose(Result=@n64) {layerGroup=None}
  ]}
 EventGraph {nodes [
   Event BlueprintUpdateAnimation()
   > n1 = Sequence!(execute) -> (then_0, then_1, then_2:Exec)
   > n2 = Gravity Override State Function DEPRECATED(CharacterPartAnimInstance=Self-Reference(), GameTimeInSeconds=GetGameTimeInSeconds(), GravityOverrideParams=Get Gravity Override Params())
   > Set GravityOverride(GravityOverride=@n2)
   > n3 = Branch(Condition=Get bIsInFrontEnd())
   > Set AccScaleOverride(AccScaleOverride:Vector="0, 0, 0") -> (Output_Get:Vector)
   > Set VelScaleOverride(VelScaleOverride:Vector="0, 0, 0") -> (Output_Get:Vector)
   n4 = Branch(execute=@n3.else, Condition=Get bIsCrouching())
   > n5 = Set AccScaleOverride(AccScaleOverride="0.1,0.1,0.05")
   n6 = Get isWearCape()
   n7 = Branch(execute=[@n4.else, @n5.then], Condition=AND Boolean(A=Get isWearBackbling(), B=NOT Boolean(A=@n6)))
   > Set AccScaleOverride(AccScaleOverride="0, 0, 0")
   > Set VelScaleOverride(VelScaleOverride="0, 0, 0")
   Set AccScaleOverride(execute=@n7.else, AccScaleOverride="0.35,0.35,0.35")
   > Set VelScaleOverride(VelScaleOverride="0.5,0.5,0.5")
   n8 = AND Boolean(A=Get isWearBackbling(), B=NOT Boolean(A=Get isWearCape()))
   n9 = Get bIsTacticalSprinting()
   Set BackPackTactJump(execute=@n1.then_2, BackPackTactJump:double=SelectFloat(A=1, B=0, bPickA=AND Boolean(A=@n9, B=@n8))) -> (Output_Get:double)
   n10 = Get bIsTargeting()
   n11 = Get bIsCrouching()
   n12 = Get bIsMoving2D()
   > Set CrouchMove(CrouchMove:double=SelectFloat(A=1, B=0, bPickA=AND Boolean!(A=@n11, B=@n12, C:bool=@n10) -> (ReturnValue))) -> (Output_Get:double)
   > Set Targeting(Targeting:double=SelectFloat(A=1, B=0, bPickA=@n10)) -> (Output_Get:double)
   n13 = AND Boolean!(A=Get isWearBackbling(), B=NOT Boolean(A=Get isWearCape())) -> (ReturnValue)
   n14 = Get bIsSkyDiving()
   > Set SkydivePack(SkydivePack:double=SelectFloat(A=1, B=0, bPickA=AND Boolean(A=@n14, B=@n13))) -> (Output_Get:double)
   Event OnCharacterCustomizationLoaded()
   n15 = CheckForBackbling()
   > Set isWearBackbling(isWearBackbling:bool=@n15) -> (Output_Get:bool)
   > Set isWearCape(isWearCape:bool=@n15.bIsBackblingLong) -> (Output_Get:bool)
   Event BlueprintInitializeAnimation()
   > Set Body_Mesh(Body_Mesh=GetOwnerSkeletalMeshComponent())
  ]}
]
variables [
 AccScaleOverride {type=Vector, default=<0.5 0.5 0.5>}
 BackPackTactJump {type=number}
 Body_Mesh {type=SkeletalMeshComponent}
 CrouchMove {type=number}
 Gravity Override Params {type=GravityOverrideParamsStruct}
 GravityOverride {type=Vector}
 hairScaleUpAlpha {type=number}
 Head_Rot_Bk {type=number}
 Head_Rot_R {type=number}
 isWearBackbling {type=boolean}
 isWearCape {type=boolean}
 SkydivePack {type=number}
 Targeting {type=number}
 VelScaleOverride {type=Vector}
]
